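; Attachment point actor for the grapple weapon, marking where a player's grapple attaches.
; On BeginPlay it takes the instigator as owner, caches player and weapon, calls the weapon's Attach, and executes the attach GameplayCue on non-authority instances.
; A tick later it binds to its attach parent's destroyed and building-died events.
; Those events, or its own destruction, make the authority detach the weapon, provided this is still its current attach point.
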(event EventBeginPlay
  (bind _returnvalue (Actor|GetAttachParentActor))
  ;[0 1 0.0615] BeginPlay
  (Utilities|FlowControl|Sequence
    (:then_0
      ;;[1 0 0.9463] Set the owner of this actor to the instigating player and cache the player and their weapon.
      (bind _instigator (Variables|Actor|GetInstigator)
        (:then
          (Actor|SetOwner _instigator)
          (bind _asplayer_pawn_athena (Utilities|Casting|CastToPlayerPawn_Athena _instigator))
          (Variables|Default|SetPlayer _asplayer_pawn_athena)
          (bind _asb_weapon_grapple_parent (Utilities|Casting|CastToB_Weapon_GrappleParent (Weapon|GetCurrentWeapon _asplayer_pawn_athena)))
          (Variables|Default|SetWeapon _asb_weapon_grapple_parent))
        (:else)))
    (:then_1
      ;;[0 0.2961 1] Notify the weapon that the attach point is valid and swinging should begin and execute the Web Hit GC.
      (bind _player (Variables|Default|GetPlayer)
        (:then
          (bind _weapon (Variables|Default|GetWeapon)
            (:then
              (Class|BWeaponGrappleParent|Attach _weapon self)
              (Networking|SwitchHasAuthority
                (:Remote
                  (GameplayCue|ExecuteGameplayCueOnActor(Burst) self (Variables|Default|GetGC_GrappleAttached) (Ability|GameplayCue|MakeGameplayCueParameters :NormalizedMagnitude 0.0 :RawMagnitude 0.0 :Location (Transformation|GetActorLocation) :Normal "0,0,0" :Instigator _player :EffectCauser _player :SourceObject _player :GameplayEffectLevel 1 :AbilityLevel 1 :bReplicateLocationWhenUsingMinimalRepProxy false)))))
            (:else)))
        (:else)))
    (:then_2
      ;;[1 0.4209 0] Detach if the attached actor is destroyed or dies. Delay until next tick to ensure the attach parent is setup properly.
      (Utilities|FlowControl|DelayUntilNextTick)
      (Utilities|IsValid _returnvalue
        (:Is%20Valid
          (Game|BindEventtoOnDestroyed _returnvalue (delegate Custom|ParentDestroyed))
          (bind _asbuilding_actor (Utilities|Casting|CastToBuildingActor _returnvalue))
          (EventDispatchers|BindEventtoOnDied _asbuilding_actor (delegate Custom|ParentDied)))))))

(event Custom|ParentDied
  (params (DamagedActor :type Actor%20Object%20Reference) (Damage :type Float%20%28single-precision%29) (InstigatedBy :type Controller%20Object%20Reference) (DamageCauser :type Actor%20Object%20Reference) (HitLocation :type Vector) (FHitComponent :type Primitive%20Component%20Object%20Reference) (BoneName :type Name) (Momentum :type Vector))
  ;[0 1 0.0615] BeginPlay
  ;;[1 0.4209 0] Detach if the attached actor is destroyed or dies. Delay until next tick to ensure the attach parent is setup properly.
  (CallFunction|TryDetach))

(event Custom|ParentDestroyed
  (params (DestroyedActor :type Actor%20Object%20Reference))
  ;[0 1 0.0615] BeginPlay
  ;;[1 0.4209 0] Detach if the attached actor is destroyed or dies. Delay until next tick to ensure the attach parent is setup properly.
  (CallFunction|TryDetach))

(event EventDestroyed
  ;[1 0.594 0.05524] Try to detach if this BGA is destroyed
  (CallFunction|TryDetach))

(event Custom|TryDetach
  ;[1 0.3103 0] If the attach point is destroyed and hasn't already detached itself from the player, detach it now.
  (Networking|SwitchHasAuthority
    (:Authority
      (bind _weapon (Variables|Default|GetWeapon)
        (:then
          (bind _attachpointbga (Variables|/GrappleParent/GrappleParent/B_Weapon_GrappleParent.B_Weapon_GrappleParent_C|GetAttachPointBGA _weapon)
            (:then
              (if (== _attachpointbga self)
                (Class|BWeaponGrappleParent|Detach (Variables|Default|GetWeapon) true true false)))
            (:else)))
        (:else)))))

(fn ConstructionScript)
