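; Parent Fortnite weapon Blueprint for grappling gear that lets a player attach a rope to a point and swing.
; Attaching draws the rope, adds an initial jump, consumes ammo and starts swinging through the controller's swinging component.
; Hit, landing, damage, water, DBNO, vehicle, skydive, zipline and unequip events drive falling, detach, animation or cooldowns.
; A repeating slower-update timer checks attachment validity and rope length; gameplay-effect changes are server-authoritative.
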(event EventBeginPlay
  (bind _returnvalue (Actor|GetComponentbyClass (Variables|Default|GetPlayer) "/Script/RidingCodeRuntime.RiderComponent"))
  (bind _player (Variables|Default|GetPlayer))
  (bind _player_2 (Variables|Default|GetPlayer))
  (bind _instigator (Variables|Actor|GetInstigator))
  ;[0 1 0] Begin Play
  (Utilities|IsValid _instigator
    (:Is%20Valid
      (bind _asplayer_pawn_athena (Utilities|Casting|CastToPlayerPawn_Athena _instigator))
      (bind _output_get (Variables|Default|SetPlayer _asplayer_pawn_athena))
      (Utilities|IsValid (Pawn|GetController _output_get)
        (:Is%20Valid
          (Variables|Default|SetSwingingController (Actor|GetComponentbyClass (Pawn|GetController _output_get) "/Script/FortniteGame.FortControllerComponent_Swinging")))
        (:Is%20Not%20Valid
          _))
      (bind _asfort_movement_comp_character (Utilities|Casting|CastToFortMovementComp_Character (Variables|/Script/Engine.Character|GetCharacterMovement (Variables|Default|GetPlayer)))
        (:then
          (Variables|Default|SetFort%20Movement%20Comp%20Character _asfort_movement_comp_character))
        (:CastFailed
          _))
      (Ability|Tasks|WaitGameplayTagAddToActor _player_2 "(TagName=\"Gameplay.Action.Player.AthenaInVehicle\")"
        (:then
          (Ability|Tasks|WaitGameplayTagAddToActor _player_2 "(TagName=\"Gameplay.Status.Rift.Skydiving\")"
            (:then
              (Ability|Tasks|WaitGameplayTagAddToActor _player_2 "(TagName=\"MovementMode.ZipliningOnSpline\")"
                (:then
                  ;; BindEvents
                  (RangedWeaponDual|BindEventtoOnHandFired (EventDispatchers|CreateEvent :!event Custom|Fired))
                  (Collision|BindEventtoOnActorHit _player (EventDispatchers|CreateEvent :!event Custom|PlayerHit))
                  (EventDispatchers|BindEventtoOnPlayerStartDBNO _player (EventDispatchers|CreateEvent :!event Custom|On%20Player%20DBNO))
                  (EventDispatchers|BindEventtoOnPawnEnterWater _player (EventDispatchers|CreateEvent :!event Custom|OnPlayerEnteredWater))
                  (Character|BindEventtoMovementModeChangedDelegate _player (EventDispatchers|CreateEvent :!event Custom|HandleMovementModeChanged))
                  (EventDispatchers|BindEventtoOnDamaged _player (EventDispatchers|CreateEvent :!event Custom|PlayerDamaged) :!class /Script/FortniteGame.FortPawn)
                  (Utilities|IsValid _returnvalue
                    (:Is%20Valid
                      (EventDispatchers|BindEventtoOnStartedRidingSingularDeprecated _returnvalue (EventDispatchers|CreateEvent :!event Custom|OnRideAnimal)))
                    (:Is%20Not%20Valid
                      _))
                  (Ability|Async|WaitGameplayEventToActor (Variables|Default|GetPlayer) "(TagName=\"Event.Athena.StopSwinging\")"
                    (:then
                      ;;[1 1 1] If the cooldown tags are empty, no need to get the Experience Settings
                      (if (Utilities|GameplayTagContainerNotEmpty (Variables|Default|GetCooldownTags))
                        (Variables|Default|SetExperienceSettingsComp (Fortnite|Athena|GameState|GetExperienceSettingsComponentfromGameState))))
                    (:EventReceived
                      (CallFunction|Fall))))
                (:Added
                  (CallFunction|OnPlayerStartedZiplining))))
            (:Added
              (CallFunction|OnPlayerSkydiving))))
        (:Added
          (CallFunction|OnPlayerEnteredVehicle))))))

(event Custom|Collision_FX_Reset
  ;[0.1195 0.6191 1] On Player Hit Checks
  ;;[0.2287 1 0.7882] Handle the player hitting something while swinging.
  ;; (Server and All Clients)
  (Utilities|FlowControl|DoOnce false
    (:Completed
      (Networking|CanExecuteCosmeticEvents
        (:True
          (Niagara|SpawnSystematLocation "/GrappleParent/GrappleParent/NS_GrappleParent_PlayerCollidedWithWall.NS_GrappleParent_PlayerCollidedWithWall" (.Location (Collision|BreakHitResult (stale HitCheck|EventOnPlayerHitWhileSwinging Hit))) (Math|Conversions|RotationFromXVector (.ImpactNormal (Collision|BreakHitResult (stale HitCheck|EventOnPlayerHitWhileSwinging Hit))))))))))

(event HitCheck|EventOnPlayerHitWhileSwinging
  (params (Hit :type Hit%20Result%20Structure%20%28by%20ref%29))
  (bind (_bblockinghit _binitialoverlap _time _distance _location _impactpoint _normal _impactnormal) (Collision|BreakHitResult Hit))
  ;[0.1195 0.6191 1] On Player Hit Checks
  ;;[0.2287 1 0.7882] Handle the player hitting something while swinging.
  ;; (Server and All Clients)
  (Utilities|IsActiveSwingingWeapon
    (:Success
      (bind _returnvalue (Class|FortAnimInputEvent|CreateAnimInputEvent "aie.Player.Swinging.PlayerHit"))
      (bind _player (Variables|Default|GetPlayer)
        (:then
          (Animation|TriggerAnimInputEvent _player _returnvalue)
          (Variables|Status|SetFreeFalling false)
          (Utilities|FlowControl|DoOnce false
            (:Completed
              (Networking|CanExecuteCosmeticEvents
                (:True
                  (Niagara|SpawnSystematLocation "/GrappleParent/GrappleParent/NS_GrappleParent_PlayerCollidedWithWall.NS_GrappleParent_PlayerCollidedWithWall" _location (Math|Conversions|RotationFromXVector _impactnormal)))))))
        (:else)))))

(event FortWeapon|EventOnEquippedAndAttached
  ;[0 1 0.2052] On Equipped
  ; (Server and All Clients)
  (bind _player (Variables|Default|GetPlayer)
    (:then
      (Weapon|SetHolsterWeapon _player 0 false false "GrappleParent")
      (CallFunction|StartSlowerUpdateTimer)
      (bind _player_2 (Variables|Default|GetPlayer)
        (:then
          (Variables|Status|SetLastEquippedTime (Fortnite|Sync|Time|GetServerWorldTimeSeconds))
          (bind _returnvalue (Class|FortAnimInputEvent|CreateAnimInputEvent "Player.Swinging.Equipped"))
          (Animation|TriggerAnimInputEvent _player_2 _returnvalue)
          (Variables|Status|SetEquipped true))
        (:else)))
    (:else)))

(event Custom|OnRideAnimal
  (params (Rider :type Rider%20Component%20Object%20Reference) (Ridable :type Ridable%20Component%20Object%20Reference))
  ;[0.3222 1 0.4776] On Started Riding
  (Utilities|IsActiveSwingingWeapon
    (:Success
      (CallFunction|EndAnimation))))

(event Custom|OnPlayerEnteredWater
  (params (FortPlayerPawn :type Fort%20Player%20Pawn%20Object%20Reference))
  ;[1 0.5567 0.0205] Detach Conditions
  ;;[1 0.4618 0.8866] Entered Water
  (Utilities|IsActiveSwingingWeapon
    (:Success
      (if (not (Default|IsSwingAttached))
        (Variables|Default|SetbPlayDetachAudio false)
        (CallFunction|ForceExitSwinging)))))

(event Custom|On%20Player%20DBNO
  ;[1 0.5567 0.0205] Detach Conditions
  ;;[0.9915 0.1511 1] DBNO
  (Utilities|IsActiveSwingingWeapon
    (:Success
      (CallFunction|ForceExitSwinging))))

(event Custom|RetryLanding
  (bind _isswinging (CallFunction|IsSwinging))
  ;[0.1195 0.6191 1] On Player Hit Checks
  ;;[0.1765 1 0.2409] Handle the player touching the ground after they have detached from swinging.
  ;; (Server and All Clients)
  (Variables|Status|SetOnHit_WasSwinging _isswinging)
  (if (Math|Boolean|ANDBoolean _isswinging (not (Variables|Default|GetbJustFired)) (not (CallFunction|ShouldBlockSwingEndonLand)))
    (bind _swingingcontroller (Variables|Default|GetSwingingController)
      (:then
        (Swinging|EndSwingingMovement _swingingcontroller))
      (:else
        (CallFunction|OnSwingingEnded false)))
    (CallFunction|EndAnimation)
    (CallFunction|SlowerUpdate)
    (Networking|SwitchHasAuthority
      (:Authority
        (Utilities|Time|SetTimerbyEvent (delegate Custom|GiveFallDamageImmunity) 0.3)))))

(event HitCheck|EventOnPlayerTouchedGroundAfterSwingEnded
  (bind _isswinging (CallFunction|IsSwinging))
  ;[0.1195 0.6191 1] On Player Hit Checks
  ;;[0.1765 1 0.2409] Handle the player touching the ground after they have detached from swinging.
  ;; (Server and All Clients)
  (CallFunction|JustTouchedGround)
  (Variables|Status|SetOnHit_WasSwinging _isswinging)
  (if (Math|Boolean|ANDBoolean _isswinging (not (Variables|Default|GetbJustFired)) (not (CallFunction|ShouldBlockSwingEndonLand)))
    (bind _swingingcontroller (Variables|Default|GetSwingingController)
      (:then
        (Swinging|EndSwingingMovement _swingingcontroller))
      (:else
        (CallFunction|OnSwingingEnded false)))
    (CallFunction|EndAnimation)
    (CallFunction|SlowerUpdate)
    (Networking|SwitchHasAuthority
      (:Authority
        (Utilities|Time|SetTimerbyEvent (delegate Custom|GiveFallDamageImmunity) 0.3)))))

(event FortWeapon|EventOnLocalUnEquip
  ;[0.8427 0.4443 0.04023] On Unequip
  ; (Server and Owning Client)
  (CallFunction|SetBuildModeBlocked false)
  (bind _player (Variables|Default|GetPlayer)
    (:then
      (if (Swinging|GetIsSwinging _player)
        (if (Pawn|IsLocallyControlled (Variables|Default|GetPlayer))
          (CallFunction|RequestDetachfromClient true true)
          (else
            (CallFunction|Detach true true)))
        (Utilities|FlowControl|Sequence
          (:then_0
            (if (Class|AthenaFunctionLibrary|RowtoBool (Variables|Default|GetSF_HolsterWeaponOnUnequipped))
              (Weapon|SetHolsterWeapon (Variables|Default|GetPlayer) 0 true false "GrappleParent")))
          (:then_1
            (if (Variables|Default|GetCancelSwingingOnUnequip)
              (CallFunction|OnUsableWeaponEquipped))))))
    (:else)))

(event Custom|GiveFallDamageImmunity
  ;[0.1195 0.6191 1] On Player Hit Checks
  ;;[0.1765 1 0.2409] Handle the player touching the ground after they have detached from swinging.
  ;; (Server and All Clients)
  (bind _player (Variables|Default|GetPlayer)
    (:then
      (GameplayEffects|ApplyGameplayEffectToSelf :self (Ability|GetFortAbilitySystemComponent _player) :GameplayEffectClass "/Game/Athena/Items/Parents/FallImmunity/GE_Athena_TillLandFallDamageImmunity_NoFX_Grant_ApplyInAirOnly.GE_Athena_TillLandFallDamageImmunity_NoFX_Grant_ApplyInAirOnly_C" :Level 0.0) ; (only applies if they're in the air)\nGives some fall damage forgiveness if player taps the \nground and then keeps going and falls off
    )
    (:else)))

(event Custom|OnPlayerEnteredVehicle
  ;[1 0.5567 0.0205] Detach Conditions
  ;;[0.5525 0.3047 1] Entered Vehicle
  (Utilities|IsActiveSwingingWeapon
    (:Success
      (CallFunction|ForceExitSwinging))))

(event Custom|OnPlayerSkydiving
  ;[1 0.5567 0.0205] Detach Conditions
  ;;[0.1407 0.1494 1] Skydiving
  (Utilities|IsActiveSwingingWeapon
    (:Success
      (CallFunction|ForceExitSwinging))))

(event Custom|ConsumeResourcesFromFiring
  ;[0 0.3802 0.1219] Attach
  ; (Server and All Clients)
  ;;[0.884 0.2185 1] Consume some ammo when fired
  ;; (Server Only)
  (Networking|SwitchHasAuthority
    (:Authority
      (FortWeapon|Ammo|ConsumeAmmo 1 true))))

(event Custom|JustTouchedGround)

(event Custom|OnPlayerStartedZiplining
  ;[1 0.5567 0.0205] Detach Conditions
  ;;[0.5525 0.3047 1] Entered Vehicle
  (Utilities|IsActiveSwingingWeapon
    (:Success
      (CallFunction|ForceExitSwinging))))

(event FortWeapon|EventOnUnEquippedAndDetached
  ;[0.8427 0.4443 0.04023] On Unequip
  ; (Server and Owning Client)
  (|CallParentFunction)
  (CallFunction|CheckAttachmentwhenUnequipped))

(event Custom|RecheckLanding
  ;[0.1195 0.6191 1] On Player Hit Checks
  (if (and (Swinging|IsSwingingGrounded (Variables|Default|GetSwingingController)) (not (Default|IsSwingAttached)))
    (CallFunction|RetryLanding)))

(event Custom|ForceExitSwinging
  ;[1 0.5567 0.0205] Detach Conditions
  (Variables|Ride|SetbBlockRiding true)
  (Variables|Status|SetFreeFalling false)
  (CallFunction|SetBuildModeBlocked false)
  (bind _swingingcontroller (Variables|Default|GetSwingingController)
    (:then
      (Swinging|EndSwingingMovement _swingingcontroller))
    (:else
      (CallFunction|OnSwingingEnded false)))
  (Networking|SwitchHasAuthority
    (:Authority
      (CallFunction|Detach true true true)
      (Development|PrintString "NYXGLASS_DETACH_forceexitswinging"))))

(event Custom|OnUsableWeaponEquipped
  ;[0.8427 0.4443 0.04023] On Unequip
  ; (Server and Owning Client)
  (if (CallFunction|IsSwinging)
    (Utilities|FlowControl|Sequence
      (:then_0
        (bind _player (Variables|Default|GetPlayer)
          (:then
            (bind _returnvalue (Class|FortAnimInputEvent|CreateAnimInputEvent "Player.Swinging.Unequipped"))
            (Animation|TriggerAnimInputEvent _player _returnvalue)
            (CallFunction|ApplyCooldown true)
            (bind _cameramodifier (Variables|Default|GetCameraModifier)
              (:then
                (Class|CameraModifierGrappleParent|TurnOff _cameramodifier))
              (:else)))
          (:else)))
      (:then_1
        (if (CallFunction|IsSwinging)
          (CallFunction|ExitSwingingWithoutExitingSwingingMovementMode)
          (Networking|SwitchHasAuthority
            (:Authority
              (CallFunction|Detach true true true)
              (Development|PrintString "NYXGLASS_DETACH_usableweapon")
              (bind _player_2 (Variables|Default|GetPlayer)
                (:then
                  (GameplayEffects|RemoveActiveEffectswithGrantedTags (Ability|GetFortAbilitySystemComponent _player_2) "(GameplayTags=((TagName=\"Granted.Athena.GrappleParent.SpeedLinesActive\")))"))
                (:else))))))
      (:then_2
        (Utilities|Time|ClearandInvalidateTimerbyHandle (Variables|SlowUpdate|GetSlowerUpdateTimer))
        (Variables|Status|SetEquipped false)))))

(event Custom|OnAttach)

(event HitCheck|EventOnPlayerStartSkidding
  ;[0.1195 0.6191 1] On Player Hit Checks
  ;;[0.501 1 0.2015] Apply the skidding GE to the player when they start swinging close to the ground.
  ;; (Server Only)
  (Networking|SwitchHasAuthority
    (:Authority
      (Variables|Status|SetGrounded true)
      (bind _player (Variables|Default|GetPlayer)
        (:then
          (GameplayEffects|ApplyGameplayEffectToSelf :self (Ability|GetFortAbilitySystemComponent _player) :GameplayEffectClass "/GrappleParent/GrappleParent/GE_GrappleParent_Skidding.GE_GrappleParent_Skidding_C" :Level 0.0))
        (:else)))))

(event Custom|Attach
  (params (Attach%20Point%20BGA :type BGA%20Grapple%20Parent%20Attachment%20Point%20Object%20Reference))
  (bind _player (Variables|Default|GetPlayer))
  (bind _returnvalue (Pawn|GetController _player))
  (bind _returnvalue_2 (Actor|GetComponentbyClass _returnvalue "/Script/FortniteGame.FortControllerComponent_Swinging"))
  (bind (_any_delay _delay) (CallFunction|DelayBetweenRopeAttachandMovement))
  (bind _player_2 (Variables|Default|GetPlayer))
  (bind _returnvalue_3 (Swinging|SwingingVectortoAttachPoint _player_2 "initial jump"))
  (bind _z (.z _returnvalue_3))
  (bind _returnvalue_4 (Math|Vector|VectorLength _returnvalue_3))
  (bind _initialjump_maxheight (Variables|Design|GetInitialJump_MaxHeight))
  (bind _returnvalue_5 (Utilities|GetValueatZero _initialjump_maxheight))
  (bind _initialjump_ropelengthformax (Variables|Design|GetInitialJump_RopeLengthForMax))
  (bind _returnvalue_6 (Utilities|GetValueatZero _initialjump_ropelengthformax))
  (bind _grounddistanceinfluenceoninitialjump (Variables|Design|GetGroundDistanceInfluenceOnInitialJump))
  (bind _returnvalue_7 (Utilities|GetValueatZero _grounddistanceinfluenceoninitialjump))
  (bind _distancetogroundonattach (Variables|Status|GetDistanceToGroundOnAttach))
  (bind _returnvalue_8 (* _distancetogroundonattach _returnvalue_7))
  (bind _returnvalue_9 (- _returnvalue_4 _z))
  (bind _returnvalue_10 (- _returnvalue_9 _returnvalue_8))
  (bind _output (Utilities|RemapFloat _returnvalue_10 (Math|Vector2D|MakeVector2D 0.0 _returnvalue_6) (Math|Vector2D|MakeVector2D 0.0 _returnvalue_5) true))
  (bind _player_3 (Variables|Default|GetPlayer))
  (bind (_velocity_x _velocity_y _velocity_z) (Math|Vector|BreakVector (Variables|/Script/Engine.MovementComponent|GetVelocity_X (Variables|/Script/Engine.Character|GetCharacterMovement (Variables|Default|GetPlayer)))))
  (bind _player_4 (Variables|Default|GetPlayer))
  (bind _player_5 (Variables|Default|GetPlayer))
  ;[0 0.3802 0.1219] Attach
  ; (Server and All Clients)
  (Utilities|IsValid Attach%20Point%20BGA
    (:Is%20Valid
      (Variables|Attach|SetAttachPointBGA Attach%20Point%20BGA)
      (Variables|Default|SetbWaitingforBGASpawn false)
      (Variables|Status|SetFreeFalling false)
      (bind _player_6 (Variables|Default|GetPlayer)
        (:then
          (Variables|/Script/FortniteGame.FortPlayerPawnAthena|SetSwinging_AttachBGA Attach%20Point%20BGA _player_6)
          (if (not (Default|IsSwingAttached))
            (Utilities|FlowControl|Sequence
              (:then_0
                ;;[0.5184 0.3176 1] Start drawing the rope
                (Rope|AttachRope (Variables|Attach|GetAttachPointBGA))
                (Networking|ForceNetUpdate)
                (Variables|Default|SetRopeAttached true)
                (Variables|Default|SetLastRopeAttachTime (Utilities|Time|GetGameTimeinSeconds)))
              (:then_1
                ;;[1 0 0.9224] Initial Jump
                ;; (Server and All Clients)
                (bind (_grounded _hitnormal _distance) (Class|PlayerPawnAthena|SwingingCheckGrounded (Variables|Default|GetPlayer)))
                (Variables|Status|SetDistanceToGroundOnAttach _distance)
                (if (or _grounded (not (CallFunction|IsSwinging)))
                  (if (> _output _velocity_z)
                    (Swinging|JumpandImpulse (Variables|Default|GetSwingingController) (Math|Vector|MakeVector 0.0 0.0 (* (Math|Float|SelectFloat _output (Math|Float|Max(Float) _output (Utilities|GetValueatZero (Variables|Design|GetInitialJump_MinHeightForFirstJump))) (CallFunction|IsSwinging)) (CallFunction|InitialJumpMultiplier)))))))
              (:then_2
                (if _any_delay
                  (Utilities|Time|SetTimerbyEvent (delegate Custom|DoAttach%20%28after%20delay%29) _delay)
                  (else
                    (Variables|Attach|SetLastAttachedTime (Utilities|Time|GetGameTimeinSeconds))
                    (Variables|Status|SetbIsPullingActor false)
                    (Variables|Default|SetbPlayDetachAudio true)
                    (Utilities|FlowControl|Sequence
                      (:then_1
                        ;;[0.2393 0 1] Record the attach time and swing vector on the player
                        ;; (Server and All Clients)
                        (Variables|/Game/Athena/PlayerPawn_Athena.PlayerPawn_Athena_C|SetSwinging_LastAttachTime (Fortnite|Sync|Time|GetServerWorldTimeSeconds) _player_3)
                        (Variables|/Game/Athena/PlayerPawn_Athena.PlayerPawn_Athena_C|SetSwinging_RightVec (Transformation|GetRightVector (Variables|/Script/Engine.Character|GetCapsuleComponent _player_3)) _player_3)
                        ;;[0.5224 0.1643 1] If the player is swinging for the first time, record the start time.
                        (if (not (CallFunction|IsSwinging))
                          (Variables|Status|SetStartedSwingingTime (Fortnite|Sync|Time|GetServerWorldTimeSeconds))
                          ;;[0.9819 0.2169 1] Clear block riding because now that we've attached to a new actor, PendingRide isn't valid.
                          (Variables|Ride|SetbBlockRiding false)))
                      (:then_3
                        ;;[0.884 0.2185 1] Consume some ammo when fired
                        ;; (Server Only)
                        (CallFunction|ConsumeResourcesfromFiring))
                      (:then_5
                        ;;[0.2226 0.5442 1] Apply the swinging GE to the player
                        ;; (Server Only)
                        (Networking|SwitchHasAuthority
                          (:Authority
                            (bind _returnvalue_11 (GameplayTags|HasMatchingGameplayTag(Message) _player_4 "(TagName=\"MovementMode.Swinging\")"))
                            (if (not _returnvalue_11)
                              (GameplayEffects|ApplyGameplayEffectToTarget :self (Ability|GetFortAbilitySystemComponent _player_4) :GameplayEffectClass (Variables|Default|GetGE_InMovementMode) :Target (Ability|GetFortAbilitySystemComponent _player_4) :Level 0.0) ; GE to signify player is in swinging movement mode
                              (CallFunction|OnEnteredSwinging(ServerOnly))))))
                      (:then_6
                        ;;[0.1415 0.2219 1] Tell the player pawn to switch to the swinging movement mode
                        ;; (Server and Owning Client)
                        (Utilities|IsValid _returnvalue
                          (:Is%20Valid
                            (Utilities|IsValid _returnvalue_2
                              (:Is%20Valid
                                (Swinging|SwingAttached _returnvalue_2 (Variables|Attach|GetAttachPointBGA)))
                              (:Is%20Not%20Valid
                                _)))
                          (:Is%20Not%20Valid
                            _))
                        (Networking|SwitchHasAuthority
                          (:Authority
                            _)
                          (:Remote
                            (if (Pawn|IsLocallyControlled (Variables|Default|GetPlayer))
                              _)))
                        (Variables|Status|SetbIsAttached true))
                      (:then_7
                        ;;[0.05392 1 0.6747] Start the swinging animation
                        ;; (Server and All Clients)
                        (bind _returnvalue_12 (Class|FortAnimInputEvent|CreateAnimInputEvent "Player.Swinging.Attach"))
                        (Animation|TriggerAnimInputEvent (Variables|Default|GetPlayer) _returnvalue_12))
                      (:then_8
                        ;;[0.1067 0.5288 1] Play Web Hit sound
                        ;; (All Clients)
                        (Networking|SwitchHasAuthority
                          (:Remote
                            (bind _returnvalue_13 (Audio|SpawnSoundatLocation (Variables|Sound|GetHitObjectSound) (Transformation|GetActorLocation (Variables|Attach|GetAttachPointBGA))))
                            (CallFunction|FillAudioComponentParams _returnvalue_13))))
                      (:then_9
                        ;;[0 0.8151 1] Force Detach in case the player detaches right after attaching - prevents potential bugs
                        ;; (Owning Client Only)
                        (if (Variables|Attach|GetbRequireHeldInputForAttach)
                          (if (Pawn|IsLocallyControlled _player_5)
                            (Utilities|FlowControl|Sequence
                              (:then_0
                                (Utilities|FlowControl|DelayUntilNextTick))
                              (:then_1
                                (Utilities|FlowControl|Delay 0.2)))
                            (GameplayEffects|RemoveActiveEffectswithGrantedTags (Ability|GetFortAbilitySystemComponent _player_5) (Variables|Default|GetBlockUseTag))
                            (if (not (Variables|Status|GetHoldingFire))
                              (CallFunction|RequestDetachfromClient true true)
                              (Development|PrintString "NYXGLASS_DETACH_detachrightafterattaching")))))
                      (:then_10
                        ;;[1 0.1864 0.867] Failsafe to detach and hide the rope if the player unequips the weapon while still attaching (possible if there's any delay between rope attachment and movement).
                        (if (ScalableFloat|GetValueasBool (Variables|Default|GetSF_DetachIfUnequippedWhileAttaching))
                          (Utilities|FlowControl|DelayUntilNextTick)
                          (if (not (or (FortWeapon|IsEquipped) (CallFunction|IsBeingEquipped)))
                            (Variables|Default|SetRopeAttached false)
                            (Rope|DetachRope)
                            (if (Pawn|IsLocallyControlled (Variables|Default|GetPlayer))
                              (CallFunction|RequestDetachfromClient true true)
                              (else
                                (CallFunction|Detach true true))))))
                      (:then_11
                        (CallFunction|OnAttach)))))))))
        (:else)))))

(event Custom|OnEnteredSwinging%20%28Server%20Only%29)

(event HitCheck|EventOnPlayerHitWallWhileSwinging
  ;[0.1195 0.6191 1] On Player Hit Checks
  ;;[0.03282 1 0.338] Handle the player hitting a wall at a low enough speed to detach.
  ;; (Server and All Clients)
  (Utilities|FlowControl|Delay 0.1)
  (bind _player (Variables|Default|GetPlayer)
    (:then
      (if (< (Math|Vector|VectorLengthSquared (Variables|/Script/Engine.MovementComponent|GetVelocity (Variables|/Script/Engine.Character|GetCharacterMovement _player))) (Math|Float|Square (Utilities|GetValueatZero (Variables|Design|GetOnHit_MinVelocityToNotDetachAfterHit))))
        (Variables|Status|SetDontShowHoldTip true)
        (bind _returnvalue (Class|FortAnimInputEvent|CreateAnimInputEvent "Player.Swinging.HitWallAndDetached"))
        (Animation|TriggerAnimInputEvent (Variables|Default|GetPlayer) _returnvalue)
        (Networking|SwitchHasAuthority
          (:Authority
            (CallFunction|Detach true)
            (Development|PrintString "NYXGLASS_DETACH_hitwall")
            (CallFunction|OnDetachwhenHitWall(Server)))
          (:Remote
            (Development|PrintString "NYXGLASS_DETACH_hitwall_client")))))
    (:else)))

(event Custom|UpdatePendingRide
  (bind _attachpointbga (Variables|Attach|GetAttachPointBGA))
  (bind _returnvalue (Actor|GetAttachParentActor _attachpointbga))
  ;[0.6765 1 0.1936] Check if the player should be starting to ride an animal.
  (Variables|Attach|GetAttachPointBGA
    (:then
      (bind _player (Variables|Default|GetPlayer)
        (:then
          (if (< (Math|Vector|VectorLengthSquared (Swinging|SwingingVectortoAttachPoint _player "checking length")) (Math|Float|Square (Utilities|GetValueatZero (Variables|Design|GetDistanceToDetach))))
            (Utilities|IsValid _returnvalue
              (:Is%20Valid
                (Utilities|IsValid (Actor|GetComponentbyClass _player "/Script/RidingCodeRuntime.RiderComponent")
                  (:Is%20Valid
                    (Utilities|IsValid (Actor|GetComponentbyClass _returnvalue "/Script/RidingCodeRuntime.RidableComponent")
                      (:Is%20Valid
                        (Variables|Ride|SetActorToRide _returnvalue)
                        (Variables|Ride|SetPendingRide true))
                      (:Is%20Not%20Valid
                        _)))
                  (:Is%20Not%20Valid
                    _)))
              (:Is%20Not%20Valid
                _))))
        (:else
          _))
      (Variables|Ride|SetPendingRide false))
    (:else)))

(event Custom|ApplyCooldown
  (params (Force :type Boolean))
  (bind _player (Variables|Default|GetPlayer))
  (bind _returnvalue (Ability|GetFortAbilitySystemComponent _player))
  (bind _freefalling (Variables|Status|GetFreeFalling))
  (bind _output (CallFunction|CalculateCooldownTime))
  (bind _cooldown_gotshotduration (Variables|Design|GetCooldown_GotShotDuration))
  (bind _returnvalue_2 (Utilities|GetValueatZero _cooldown_gotshotduration))
  (bind _returnvalue_3 (Math|Float|SelectFloat _returnvalue_2 _output _freefalling))
  ;[0.1012 1 0.9303] Apply Cooldown
  (Utilities|IsActiveSwingingWeapon
    (:Success
      (Networking|SwitchHasAuthority
        (:Authority
          (bind _player_2 (Variables|Default|GetPlayer)
            (:then
              (if (not (GameplayTags|HasMatchingGameplayTag _player_2 "(TagName=\"Granted.Cooldown.GrappleParent\")"))
                (bind _returnvalue_4 (Ability|GameplayEffect|AssignTagSetbyCallerMagnitude (GameplayEffects|MakeOutgoingSpec :self _returnvalue :GameplayEffectClass "/GrappleParent/GrappleParent/GE_GrappleParent_Cooldown_Variable.GE_GrappleParent_Cooldown_Variable_C" :Level 0.0) "(TagName=\"SetByCaller.Duration\")" _returnvalue_3))
                (GameplayEffects|ApplyGameplayEffectSpecToSelf _returnvalue _returnvalue_4)
                (CallFunction|CallReticleCooldown _returnvalue_3)
                (CallFunction|OnCooldownApplied)))
            (:else)))))))

(event Custom|PlayerDamaged
  (params (DamagedActor :type Actor%20Object%20Reference) (Damage :type Float%20%28single-precision%29) (InstigatedBy :type Controller%20Object%20Reference) (DamageCauser :type Actor%20Object%20Reference) (HitLocation :type Vector) (FHitComponent :type Primitive%20Component%20Object%20Reference) (BoneName :type Name) (Momentum :type Vector))
  ;[1 0.5567 0.0205] Detach Conditions
  ;;[0.04791 0.413 1] Player shot while swinging
  ;; (Server Only - plays sound on all clients)
  (Utilities|IsActiveSwingingWeapon
    (:Success
      (Networking|SwitchHasAuthority
        (:Authority
          (Utilities|IsValid InstigatedBy
            (:Is%20Valid
              (Utilities|Casting|CastToFortWeapon DamageCauser)
              (if (Math|Boolean|ANDBoolean (CallFunction|IsSwinging) (Variables|Default|GetFallWhenShot) (not (Variables|Status|GetFreeFalling)))
                (CallFunction|Fall)
                (CallFunction|DamagedMulticast HitLocation)))))))))

(event Custom|OnDetachWhenHitWall%20%28Server%29)

(event Custom|CallReticleCooldown
  (params (Cooldown%20Duration :type Float%20%28double-precision%29))
  ;[0.1012 1 0.9303] Apply Cooldown
  (Default|CallReticleCooldown Cooldown%20Duration (Utilities|GetValueatZero (Variables|Design|GetCooldown_NormalDuration))))

(event Custom|DoAttach%20%28after%20delay%29
  (bind _player (Variables|Default|GetPlayer))
  (bind _returnvalue (Pawn|GetController _player))
  (bind _returnvalue_2 (Actor|GetComponentbyClass _returnvalue "/Script/FortniteGame.FortControllerComponent_Swinging"))
  (bind _player_2 (Variables|Default|GetPlayer))
  (bind _player_3 (Variables|Default|GetPlayer))
  (bind _player_4 (Variables|Default|GetPlayer))
  ;[0 0.3802 0.1219] Attach
  ; (Server and All Clients)
  (Variables|Attach|SetLastAttachedTime (Utilities|Time|GetGameTimeinSeconds))
  (Variables|Status|SetbIsPullingActor false)
  (Variables|Default|SetbPlayDetachAudio true)
  (Utilities|FlowControl|Sequence
    (:then_1
      ;;[0.2393 0 1] Record the attach time and swing vector on the player
      ;; (Server and All Clients)
      (Variables|/Game/Athena/PlayerPawn_Athena.PlayerPawn_Athena_C|SetSwinging_LastAttachTime (Fortnite|Sync|Time|GetServerWorldTimeSeconds) _player_2)
      (Variables|/Game/Athena/PlayerPawn_Athena.PlayerPawn_Athena_C|SetSwinging_RightVec (Transformation|GetRightVector (Variables|/Script/Engine.Character|GetCapsuleComponent _player_2)) _player_2)
      ;;[0.5224 0.1643 1] If the player is swinging for the first time, record the start time.
      (if (not (CallFunction|IsSwinging))
        (Variables|Status|SetStartedSwingingTime (Fortnite|Sync|Time|GetServerWorldTimeSeconds))
        ;;[0.9819 0.2169 1] Clear block riding because now that we've attached to a new actor, PendingRide isn't valid.
        (Variables|Ride|SetbBlockRiding false)))
    (:then_3
      ;;[0.884 0.2185 1] Consume some ammo when fired
      ;; (Server Only)
      (CallFunction|ConsumeResourcesfromFiring))
    (:then_5
      ;;[0.2226 0.5442 1] Apply the swinging GE to the player
      ;; (Server Only)
      (Networking|SwitchHasAuthority
        (:Authority
          (bind _returnvalue_3 (GameplayTags|HasMatchingGameplayTag(Message) _player_3 "(TagName=\"MovementMode.Swinging\")"))
          (if (not _returnvalue_3)
            (GameplayEffects|ApplyGameplayEffectToTarget :self (Ability|GetFortAbilitySystemComponent _player_3) :GameplayEffectClass (Variables|Default|GetGE_InMovementMode) :Target (Ability|GetFortAbilitySystemComponent _player_3) :Level 0.0) ; GE to signify player is in swinging movement mode
            (CallFunction|OnEnteredSwinging(ServerOnly))))))
    (:then_6
      ;;[0.1415 0.2219 1] Tell the player pawn to switch to the swinging movement mode
      ;; (Server and Owning Client)
      (Utilities|IsValid _returnvalue
        (:Is%20Valid
          (Utilities|IsValid _returnvalue_2
            (:Is%20Valid
              (Swinging|SwingAttached _returnvalue_2 (Variables|Attach|GetAttachPointBGA)))
            (:Is%20Not%20Valid
              _)))
        (:Is%20Not%20Valid
          _))
      (Networking|SwitchHasAuthority
        (:Authority
          _)
        (:Remote
          (if (Pawn|IsLocallyControlled (Variables|Default|GetPlayer))
            _)))
      (Variables|Status|SetbIsAttached true))
    (:then_7
      ;;[0.05392 1 0.6747] Start the swinging animation
      ;; (Server and All Clients)
      (bind _returnvalue_4 (Class|FortAnimInputEvent|CreateAnimInputEvent "Player.Swinging.Attach"))
      (Animation|TriggerAnimInputEvent (Variables|Default|GetPlayer) _returnvalue_4))
    (:then_8
      ;;[0.1067 0.5288 1] Play Web Hit sound
      ;; (All Clients)
      (Networking|SwitchHasAuthority
        (:Remote
          (bind _returnvalue_5 (Audio|SpawnSoundatLocation (Variables|Sound|GetHitObjectSound) (Transformation|GetActorLocation (Variables|Attach|GetAttachPointBGA))))
          (CallFunction|FillAudioComponentParams _returnvalue_5))))
    (:then_9
      ;;[0 0.8151 1] Force Detach in case the player detaches right after attaching - prevents potential bugs
      ;; (Owning Client Only)
      (if (Variables|Attach|GetbRequireHeldInputForAttach)
        (if (Pawn|IsLocallyControlled _player_4)
          (Utilities|FlowControl|Sequence
            (:then_0
              (Utilities|FlowControl|DelayUntilNextTick))
            (:then_1
              (Utilities|FlowControl|Delay 0.2)))
          (GameplayEffects|RemoveActiveEffectswithGrantedTags (Ability|GetFortAbilitySystemComponent _player_4) (Variables|Default|GetBlockUseTag))
          (if (not (Variables|Status|GetHoldingFire))
            (CallFunction|RequestDetachfromClient true true)
            (Development|PrintString "NYXGLASS_DETACH_detachrightafterattaching")))))
    (:then_10
      ;;[1 0.1864 0.867] Failsafe to detach and hide the rope if the player unequips the weapon while still attaching (possible if there's any delay between rope attachment and movement).
      (if (ScalableFloat|GetValueasBool (Variables|Default|GetSF_DetachIfUnequippedWhileAttaching))
        (Utilities|FlowControl|DelayUntilNextTick)
        (if (not (or (FortWeapon|IsEquipped) (CallFunction|IsBeingEquipped)))
          (Variables|Default|SetRopeAttached false)
          (Rope|DetachRope)
          (if (Pawn|IsLocallyControlled (Variables|Default|GetPlayer))
            (CallFunction|RequestDetachfromClient true true)
            (else
              (CallFunction|Detach true true))))))
    (:then_11
      (CallFunction|OnAttach))))

(event Custom|Damaged_Multicast
  (params (HitLocation :type Vector))
  ;[1 0.5567 0.0205] Detach Conditions
  ;;[0.04791 0.413 1] Player shot while swinging
  ;; (Server Only - plays sound on all clients)
  (Networking|SwitchHasAuthority
    (:Remote
      (Audio|PlaySoundatLocation (Variables|Sound|GetRopeBreakSound) HitLocation))))

(event Custom|OnCooldownApplied)

(event Custom|UpdateReticleFromNative
  (params (HitSomething :type Boolean) (BlockedUse :type Boolean) (IsTargetAtAimLocation :type Boolean) (TargetPosition :type Vector))
  (bind (_blockedbytag _blockedbynoammo) (CallFunction|IsFiringBlocked))
  ;[1 0.3944 0.8908] Reticle Update
  (bind _output_get (Variables|Targeting|SetTargetingSomething HitSomething))
  (Variables|Default|GetPlayer
    (:then
      (EventDispatchers|CallBPReticleUpdate _output_get TargetPosition (or (not BlockedUse) (GameplayTags|HasMatchingGameplayTag (Variables|Default|GetPlayer) "(TagName=\"Granted.Athena.GrappleParent.BlockUseAfterDetach\")")) false (CallFunction|CalculateChargeTime) IsTargetAtAimLocation (Default|IsSwingAttached) (or _blockedbytag _blockedbynoammo))
      (Utilities|FlowControl|Sequence
        (:then_0
          (if (or (Variables|Targeting|GetTargetingSomething) (Variables|Default|GetHasCustomTargeting))
            (CallFunction|TrytoFire)))
        (:then_1
          (if (not (CallFunction|PlayerInputWantsAttachment))
            (Variables|Status|SetHasDetachedSinceFireWasPressed false)))))
    (:else)))

(event Custom|StopSkidding
  ;[0.6123 0 1] Remove the skidding GE if it's applied
  (Networking|SwitchHasAuthority
    (:Authority
      (Variables|Status|SetGrounded false)
      (GameplayEffects|RemoveActiveEffectswithGrantedTags (Ability|GetFortAbilitySystemComponent (Variables|Default|GetPlayer)) "(GameplayTags=((TagName=\"Granted.Athena.GrappleParent.Skidding\")))"))))

(event Custom|End%20Animation
  ;[1 0.509 0.06308] Helper function to end all swinging animation.
  (bind _player (Variables|Default|GetPlayer)
    (:then
      (Animation|MulticastUnlinkAnimLayer _player (Variables|Default|GetSwingingAnimLayerClass))
      (CallFunction|ResetMeshTransform))
    (:else)))

(event Custom|Client_ShowTip
  (params (TipTag :type Gameplay%20Tag%20Structure))
  ;[0.2842 0.2266 1] Helper function to show a tip from the server.
  (Utilities|ShowTip TipTag))

(event Custom|TEMP%20Print%20debug%20points
  ;[0 1 0.9484] Slower Update Timer
  ; (Server and All Clients)
  (Utilities|DoOncePerFrame
    (:%20%20
      (Rendering|Debug|DrawDebugPoint (+ (Transformation|GetActorLocation (Variables|Default|GetPlayer)) (Math|Vector|SelectVector "0, 0, 0" "0.000000,0.000000,10.000000" (Networking|IsServer))) 1e+01 (Math|Color|SelectColor "(R=1.000000,G=0.000000,B=0.969006,A=1.000000)" "(R=0.000000,G=1.000000,B=0.070967,A=1.000000)" (Networking|IsServer)) 6e+01))))

(event Custom|StartSlowerUpdateTimer
  ;[0 1 0.9484] Slower Update Timer
  ; (Server and All Clients)
  (bind _returnvalue (Utilities|Time|SetTimerbyEvent (EventDispatchers|CreateEvent :!event Custom|SlowerUpdate) (Utilities|GetValueatZero (Variables|SlowUpdate|GetSlowUpdateDeltaTime)) true))
  (Variables|SlowUpdate|SetSlowerUpdateTimer _returnvalue) ; clears on unequip
)

(event Custom|Fired
  (params (Hand :type EDualWeaponHand%20Enum :default "LEFT") (bPersistantFire :type Boolean)))

(event EventOnPlayWeaponFireFX
  (params (bPersistentFire :type Boolean) (bSecondaryFire :type Boolean)))

(event Custom|SlowerUpdate
  (bind _player (Variables|Default|GetPlayer))
  (bind _returnvalue (Swinging|SwingingVectortoAttachPoint _player "checking length"))
  (bind _returnvalue_2 (Math|Vector|VectorLengthSquared _returnvalue))
  (bind _swinginganimlayerclass (Variables|Default|GetSwingingAnimLayerClass))
  (bind _player_2 (Variables|Default|GetPlayer))
  (bind _mesh (Variables|/Script/Engine.Character|GetMesh _player_2))
  (bind _returnvalue_3 (Components|SkeletalMesh|AnimationBlueprintLinking|GetLinkedAnimLayerInstancebyClass _mesh _swinginganimlayerclass))
  (bind _returnvalue_4 (Ability|GameplayCue|MakeGameplayCueParameters 0.0))
  (bind _player_3 (Variables|Default|GetPlayer))
  (bind _returnvalue_5 (Ability|GetFortAbilitySystemComponent _player_3))
  (bind _player_4 (Variables|Default|GetPlayer))
  (bind _player_5 (Variables|Default|GetPlayer))
  ;[0 1 0.9484] Slower Update Timer
  ; (Server and All Clients)
  (Variables|Default|GetPlayer
    (:then
      (Utilities|FlowControl|Sequence
        (:then_0)
        (:then_1
          (Networking|SwitchHasAuthority
            (:Authority
              (if (and (Default|IsSwingAttached) (CallFunction|IsSwinging))
                ;;[0.7717 0.3236 1] Detach if the Attach Point BGA is not valid.
                (Variables|Attach|GetAttachPointBGA
                  (:then
                    (Utilities|FlowControl|Sequence
                      (:then_0
                        ;;[1 0.1244 0.5967] Detach if the rope is too long or too short.
                        (if (> _returnvalue_2 (Math|Float|Square (Utilities|GetValueatZero (Variables|Design|GetMaxWebLengthBeforeForcedDetach))))
                          (CallFunction|Detach true)
                          (Development|PrintString "NYXGLASS_DETACH_ropetoolong")
                          (elif (< _returnvalue_2 (Math|Float|Square (Utilities|GetValueatZero (Variables|Design|GetDistanceToDetach))))
                            ;;[0.08618 1 0.631] Detach if pending ride or rope is too short.
                            (CallFunction|UpdatePendingRide)
                            (if (Variables|Ride|GetPendingRide)
                              (CallFunction|Detach true true true)
                              (Development|PrintString "NYXGLASS_DETACH_pendingride")
                              (bind _swingingcontroller (Variables|Default|GetSwingingController)
                                (:then
                                  (Swinging|EndSwingingMovement _swingingcontroller))
                                (:else
                                  (CallFunction|OnSwingingEnded false)))
                              (else
                                (CallFunction|Detach true)
                                (Development|PrintString "NYXGLASS_DETACH_notpendingride"))))))
                      (:then_1
                        ;;[0.6159 0.1124 1] Update how long the player has not been giving acceleration input.
                        (if (< (Math|Vector|VectorLengthSquared (Pawn|Components|CharacterMovement|GetCurrentAcceleration (Variables|/Script/Engine.Character|GetCharacterMovement (Variables|Default|GetPlayer)))) 1e+02)
                          (Math|Integer|IncrementInt (Variables|Status|GetTimeNotGivingInput))
                          (else
                            (Variables|Status|SetTimeNotGivingInput 0))))
                      (:then_2
                        ;;[0.2591 0 1] Update how long the player has been moving slowly.
                        (if (< (Math|Vector|VectorLengthSquared (CallFunction|GetTrueVelocityBasedonTransform)) (Math|Float|Square (Utilities|GetValueatZero (Variables|Design|GetMinSpeedToNotDetachAtBadAngle))))
                          (Math|Integer|IncrementInt (Variables|Status|GetTimeBeenSlow))
                          (else
                            (Variables|Status|SetTimeBeenSlow 0))))
                      (:then_3
                        ;;[0 0.000306 1] If the player is moving too slow, detach. If the player was moving slow because of a lack of input, show the input tip now.
                        (if (>= (Variables|Status|GetTimeBeenSlow) (Utilities|GetValueatZero (Variables|Design|GetSlowUpdateTicksNeededForDetachment)))
                          (if (and (< (Math|Vector|DotProduct (Math|Vector|Normalize (Swinging|SwingingVectortoAttachPoint (Variables|Default|GetPlayer) "check sideways detach") 0.01) "0, 0,1.000000") (Utilities|Select :Index (Math|Integer|Clamp(Integer) (- (Variables|Status|GetTimeBeenSlow) (ScalableFloat|GetValueasInteger (Variables|Design|GetSlowUpdateTicksNeededForDetachment))) 0 2) :Option%200 (ScalableFloat|GetValueatLevel (Variables|Design|GetDetachmentDotsAtLowSpeeds)) :Option%201 (ScalableFloat|GetValueatLevel (Variables|Design|GetDetachmentDotsAtLowSpeeds) 1.0) :Option%202 (ScalableFloat|GetValueatLevel (Variables|Design|GetDetachmentDotsAtLowSpeeds) 2.0))) (CallFunction|ShouldDetachifTooSlow))
                            (Utilities|FlowControl|Sequence
                              (:then_0
                                (if (> (Variables|Status|GetTimeNotGivingInput) 8)
                                  (CallFunction|ClientShowTip "(TagName=\"Athena.Ability.Failed.WestSausage.NotHoldingForward\")")
                                  (Variables|Status|SetTimeNotGivingInput -20)))
                              (:then_1
                                (Networking|SwitchHasAuthority
                                  (:Remote
                                    (if (Pawn|IsLocallyControlled (Variables|Default|GetPlayer))
                                      (CallFunction|RequestDetachfromClient false)
                                      (Development|PrintString "NYXGLASS_DETACH_tooslow_parent_client")
                                      (CallFunction|OnDetachedfromTooSlow(OwningClientOnly))))))))))
                      (:then_4
                        ;;[0 0.3541 1] Detach if the actor that the AttachPoint is attached to moves.
                        (if (not (ScalableFloat|GetValueasBool (Variables|Design|GetCanAttachToMovingObjects)))
                          (bind _actorattachedto (Variables|Targeting|GetActorAttachedTo)
                            (:then
                              (if (> (Math|Vector|VectorLengthSquared (- (Transformation|GetActorLocation _actorattachedto) (Variables|Targeting|GetActorAttachedTo_OriginalPosition))) 5e+01)
                                (CallFunction|Detach true)
                                (Development|PrintString "NYXGLASS_DETACH_frommovement")))
                            (:else))))))
                  (:else
                    (if (FortWeapon|IsEquipped)
                      (CallFunction|Detach true true)
                      (Development|PrintString "NYXGLASS_DETACH_BGA_invalid"))))))))
        (:then_2
          ;;[0 0.9287 1] If the player is swinging but doesn't have the swinging movement mode applied, apply it now. If they are not swinging but do have the movement mode applied, remove it now.
          (Networking|SwitchHasAuthority
            (:Authority
              (bind _returnvalue_6 (GameplayTags|HasMatchingGameplayTag(Message) _player_3 "(TagName=\"MovementMode.Swinging\")"))
              (if (CallFunction|IsSwinging)
                (if (not _returnvalue_6)
                  (GameplayEffects|ApplyGameplayEffectToTarget :self _returnvalue_5 :GameplayEffectClass (Variables|Default|GetGE_InMovementMode) :Target _returnvalue_5 :Level 0.0)
                  (GameplayCue|AddGameplayCueOnActor(Looping) (Variables|Default|GetPlayer) (Variables|Default|GetGC_SwingingLoop) _returnvalue_4))
                (elif _returnvalue_6
                  (GameplayEffects|RemoveActiveEffectswithGrantedTags _returnvalue_5 "(GameplayTags=((TagName=\"MovementMode.Swinging\")))")
                  (GameplayCue|RemoveGameplayCueOnActor(Looping) (Variables|Default|GetPlayer) (Variables|Default|GetGC_SwingingLoop) _returnvalue_4))))))
        (:then_3
          ;;[0 1 0.1052] If the player is not swinging and has the skidding GE applied, remove it now.
          (Networking|SwitchHasAuthority
            (:Authority
              (if (not (CallFunction|IsSwinging))
                (bind _returnvalue_7 (GameplayTags|HasMatchingGameplayTag(Message) _player_5 "(TagName=\"Granted.Athena.GrappleParent.Skidding\")"))
                (if _returnvalue_7
                  (GameplayEffects|RemoveActiveEffectswithGrantedTags (Ability|GetFortAbilitySystemComponent _player_5) "(GameplayTags=((TagName=\"Granted.Athena.GrappleParent.Skidding\")))"))))))
        (:then_4
          ;;[0 1 0.7299] If the player is not swinging and has the going fast GE applied, remove it now.
          (Networking|SwitchHasAuthority
            (:Authority
              (if (not (CallFunction|IsSwinging))
                (bind _returnvalue_8 (GameplayTags|HasMatchingGameplayTag(Message) _player_4 "(TagName=\"Granted.Athena.GrappleParent.SpeedLinesActive\")"))
                (if _returnvalue_8
                  (GameplayEffects|RemoveActiveEffectswithGrantedTags (Ability|GetFortAbilitySystemComponent _player_4) "(GameplayTags=((TagName=\"Granted.Athena.GrappleParent.SpeedLinesActive\")))"))))))
        (:then_5
          ;;[1 0.1676 0.7849] Failsafe to make sure animation and camera effects properly turn on/off based on if player is swinging or not
          (Networking|SwitchHasAuthority
            (:Remote
              (CallFunction|UpdatePendingRide)
              (if (and (Swinging|GetIsSwinging (Variables|Default|GetPlayer)) (not (Variables|Ride|GetPendingRide)))
                (Utilities|IsValid _returnvalue_3
                  (:Is%20Not%20Valid
                    (if (FortWeapon|IsEquipped)
                      (CallFunction|UpdateAnimLayer true))))
                (else
                  (Utilities|IsValid _returnvalue_3
                    (:Is%20Valid
                      (if (not (Variables|Status|GetFreeFalling))
                        (CallFunction|UpdateAnimLayer false)
                        (CallFunction|ResetMeshTransform)))))))))
        (:then_6
          ;;[0.4991 0.2584 1] Failsafe for removing the rope on the owning client. Only remove it if the player is not swinging and the rope is not supposed to be attached.
          (Networking|SwitchHasAuthority
            (:Remote
              (if (Math|Boolean|ANDBoolean (not (CallFunction|IsSwinging)) (Pawn|IsLocallyControlled (Variables|Default|GetPlayer)) (not (Variables|Default|GetRopeAttached)))
                (Rope|DetachRope)))))))
    (:else
      ;;[0.3823 0.1255 1] Detach if the player is invalid.
      (Networking|SwitchHasAuthority
        (:Authority
          (CallFunction|Detach true true)
          (Development|PrintString "NYXGLASS_DETACH_invalid_player"))))))

(event Rope|EventOnRopeAttached
  (bind _ropefadematerialparamname (Variables|Rope|GetRopeFadeMaterialParamName))
  (bind _rope (Variables|Rope|GetRope))
  (bind _ropeanchor (Variables|Rope|GetRopeAnchor))
  ;[0 0.3438 1] Rope Drawing
  (Utilities|FlowControl|Sequence
    (:then_0
      (Transformation|SetWorldScale3D (Variables|Rope|GetRopeAnchor) "1.000000,1.000000,1.000000"))
    (:then_1
      (Utilities|IsValid _rope
        (:Is%20Valid
          (Rendering|Material|SetScalarParameterValueonMaterials _rope _ropefadematerialparamname 0.0))))
    (:then_2
      (Utilities|IsValid _ropeanchor
        (:Is%20Valid
          (Rendering|Material|SetScalarParameterValueonMaterials _ropeanchor _ropefadematerialparamname 0.0))))))

(event Custom|RequestDetachFromClient
  (params (DetachFromHip :type Boolean) (DontApplyForce :type Boolean) (ForceRun :type Boolean))
  ;[1 0.5552 0] Detach
  ; (Called from Server - broadcasts to Owning Client)
  (Networking|SwitchHasAuthority
    (:Remote
      (CallFunction|DetachInternal DetachFromHip DontApplyForce ForceRun)
      (CallFunction|ServerClientRequestedDetach DetachFromHip DontApplyForce ForceRun))))

(event Custom|OnSwingingEnded
  (params (SkipRemovingSwingingGE :type Boolean))
  (bind _player (Variables|Default|GetPlayer))
  (bind _returnvalue (Fortnite|GetFortPlayerControllerfromActor _player))
  ;[1 0.3468 0] On Exit Swinging
  ; (Server and All Clients)
  (Utilities|IsActiveSwingingWeapon
    (:Success
      (if (not (or (and (Default|IsSwingAttached) (Swinging|GetIsSwinging (Variables|Default|GetPlayer))) (Variables|Default|GetbJustFired)))
        (Utilities|FlowControl|Sequence
          (:then_0
            ;;[0.1332 0.5271 1] Remove active GEs from player and apply cooldown if swing exit was successful. Remove the weapon from the players inventory if they have no more ammunition.
            (Networking|SwitchHasAuthority
              (:Authority
                (bind _player_2 (Variables|Default|GetPlayer)
                  (:then
                    (if (not SkipRemovingSwingingGE)
                      (GameplayEffects|RemoveActiveEffectswithGrantedTags (Ability|GetFortAbilitySystemComponent _player_2) "(GameplayTags=((TagName=\"MovementMode.Swinging\")))"))
                    (CallFunction|StopSkidding)
                    (GameplayEffects|RemoveActiveEffectswithGrantedTags (Ability|GetFortAbilitySystemComponent _player_2) "(GameplayTags=((TagName=\"Granted.Athena.GrappleParent.BlockUseAfterDetach\")))") ; failsafe
                    (Utilities|FlowControl|Sequence
                      (:then_0
                        (Utilities|FlowControl|Delay 0.25)
                        (if (not (CallFunction|IsSwinging))
                          (CallFunction|ApplyCooldown (or (Variables|Status|GetFreeFalling) (GameplayTags|HasMatchingGameplayTag (Variables|Default|GetPlayer) "(TagName=\"MovementMode.Driving\")")))
                          (CallFunction|StopSkidding) ; Failsafe
                        ))
                      (:then_1
                        (if (<= (FortWeapon|GetMagazineAmmoCount) 0)
                          (Utilities|IsValid _returnvalue
                            (:Is%20Valid
                              (Utilities|FlowControl|Delay 0.75) ; make sure it has time to clean up the swinging
                              (CallFunction|ClientonDestroyedonDepleted)
                              (Inventory|RemoveItemFromPlayerByGuid _returnvalue (FortWeapon|GetInventoryGUID) -1 true)
                              (bind _player_3 (Variables|Default|GetPlayer)
                                (:then
                                  (Weapon|SetHolsterWeapon _player_3 0 false true "GrappleParent"))
                                (:else))))))))
                  (:else)))))
          (:then_1
            ;;;[0 1 0.9708] Unblock build mode
            (CallFunction|MulticastSetBuildModeBlocked false))
          (:then_2
            ;;[0.2668 1 0.7244] Skydiving may be pending if swinging onto a launch pad, so reset it here.
            (bind _player_4 (Variables|Default|GetPlayer)
              (:then
                (if (Parachute|GetPendingSkydiveLaunch _player_4)
                  (AthenaMovement|SetPendingSkydiveLaunch _player_4 false)))
              (:else)))
          (:then_3
            ;;[1 0.52 0.2025] End Animation
            (if (not (and (Variables|Status|GetFreeFalling) (not (Variables|Ride|GetPendingRide))))
              (CallFunction|EndAnimation)))
          (:then_4
            ;;[0.1314 0.6733 1] Hold Tip
            (Networking|SwitchHasAuthority
              (:Remote
                (if (Variables|Status|GetDontShowHoldTip)
                  (Variables|Status|SetDontShowHoldTip false)
                  (elif (< (- (Fortnite|Sync|Time|GetServerWorldTimeSeconds) (Variables|Status|GetStartedSwingingTime)) (Utilities|GetValueatZero (Variables|Design|GetSwingingTimeForTip)))
                    (Utilities|ShowTip "(TagName=\"Athena.Ability.Failed.WestSausage.DidntHold\")"))))))
          (:then_5
            ;;[0.07192 0.3039 1] Unholster the weapon and disable the high acceleration tolerance.
            (bind _player_5 (Variables|Default|GetPlayer)
              (:then
                (Weapon|SetHolsterWeapon _player_5 0 false true "GrappleParent"))
              (:else
                _))
            (Rope|DetachRope)
            (bind _attachpointbga (Variables|Attach|GetAttachPointBGA)
              (:then
                (Actor|DestroyActor _attachpointbga)
                (Networking|SwitchHasAuthority
                  (:Authority
                    (NetTolerance|SetUseHighAccelerationTolerance (Variables|Default|GetFort%20Movement%20Comp%20Character) false "GrappleGloves"))))
              (:else)))
          (:then_6
            ;;[0.1384 0.7888 1] Start riding an actor if we are pending ride.
            (if (Math|Boolean|ANDBoolean (Variables|Ride|GetPendingRide) (not (Variables|Ride|GetbBlockRiding)) (< (Transformation|GetSquaredDistanceTo (Variables|Default|GetPlayer) (Variables|Ride|GetActorToRide)) (Math|Float|Square 1e+03)))
              (Variables|Ride|SetPendingRide false)
              (Riding|StartRidingActor (Actor|GetComponentbyClass (Variables|Default|GetPlayer) "/Script/RidingCodeRuntime.RiderComponent") (Variables|Ride|GetActorToRide) 0)))
          (:then_7
            (if (not (Variables|Status|GetFreeFalling))
              ;;[0.3267 0.3819 1] Failsafe to ensure the removal of the swinging animation layer.
              (CallFunction|UpdateAnimLayer false)
              (Variables|Default|SetbPlayDetachAudio false))))))))

(event Custom|ClientOnDestroyedOnDepleted
  ;[1 0.3468 0] On Exit Swinging
  ; (Server and All Clients)
  ;;[1 1 1] Sound on removed from inventory
  (Networking|CanExecuteCosmeticEvents
    (:True
      (bind _player (Variables|Default|GetPlayer)
        (:then
          (if (Pawn|IsLocallyControlled _player)
            (Audio|PlaySound2D (Variables|Sound|GetWeaponDepletedDiscardSound))))
        (:else)))))

(event Rope|EventOnRopeUpdated
  (params (RopeDirection :type Vector :default "0, 0, 0") (RopeLength :type Float%20%28single-precision%29))
  (bind _ropeanchor (Variables|Rope|GetRopeAnchor))
  ;[0 0.3438 1] Rope Drawing
  (Utilities|IsValid _ropeanchor
    (:Is%20Valid
      (Rendering|Material|SetVectorParameterValueonMaterials _ropeanchor (Variables|Rope|GetRopeAnchorOffsetMaterialParamName) (* RopeDirection (Math|Float|Min(Float) RopeLength (Variables|Rope|GetMaxAnchorDistanceFromAttachPoint)))))))

(event Custom|Server_ClientRequestedDetach
  (params (DetachFromHip :type Boolean) (DontApplyForce :type Boolean) (ForceRun :type Boolean))
  ;[1 0.5552 0] Detach
  ; (Called from Server - broadcasts to Owning Client)
  (Networking|SwitchHasAuthority
    (:Authority
      (CallFunction|DetachInternal DetachFromHip DontApplyForce ForceRun))))

(event Custom|ExitSwingingWithoutExitingSwingingMovementMode
  ;[1 0.3468 0] On Exit Swinging
  ; (Server and All Clients)
  (CallFunction|OnSwingingEnded true))

(event Rope|EventOnRopeStartFadeOut
  (bind _returnvalue (Utilities|Time|GetTimeSeconds))
  (bind _rope (Variables|Rope|GetRope))
  (bind _ropeanchor (Variables|Rope|GetRopeAnchor))
  (bind _ropefadematerialparamname (Variables|Rope|GetRopeFadeMaterialParamName))
  (bind _ropetimestampmaterialparamname (Variables|Rope|GetRopeTimestampMaterialParamName))
  ;[0 0.3438 1] Rope Drawing
  (Utilities|FlowControl|Sequence
    (:then_0
      (Utilities|IsValid _rope
        (:Is%20Valid
          (Rendering|Material|SetScalarParameterValueonMaterials _rope _ropetimestampmaterialparamname _returnvalue)
          (Rendering|Material|SetScalarParameterValueonMaterials _rope _ropefadematerialparamname 1.0))))
    (:then_1
      (Utilities|IsValid _ropeanchor
        (:Is%20Valid
          (Rendering|Material|SetScalarParameterValueonMaterials _ropeanchor _ropetimestampmaterialparamname _returnvalue)
          (Rendering|Material|SetScalarParameterValueonMaterials _ropeanchor _ropefadematerialparamname 1.0))))))

(event Custom|Detach
  (params (DetachFromHip :type Boolean) (DontApplyForce :type Boolean) (ForceRun :type Boolean))
  ;[1 0.5552 0] Detach
  ; (Called from Server - broadcasts to Owning Client)
  (Networking|SwitchHasAuthority
    (:Authority
      (CallFunction|DetachClient DetachFromHip DontApplyForce ForceRun)
      (CallFunction|DetachInternal DetachFromHip DontApplyForce ForceRun))))

(event Custom|OnDetachedFromTooSlow%20%28Owning%20Client%20Only%29)

(event Custom|HandleMovementModeChanged
  (params (Character :type Character%20Object%20Reference) (PrevMovementMode :type EMovementMode%20Enum :default "MOVE_None") (PreviousCustomMode :type Byte))
  ;[1 0.3468 0] On Exit Swinging
  ; (Server and All Clients)
  ;;[1 0.2088 0] Handle the player's movement mode changing. If they were previously Swinging or SwingingGrounded, and are currently not Swinging or SwingingGrounded, exit all swinging movement.
  (Utilities|IsActiveSwingingWeapon
    (:Success
      (if (and (Utilities|Enum|Equal(Enum) PrevMovementMode "MOVE_Custom") (or (Utilities|Enum|Equal(Enum) (Variables|Default|GetSwingingCustomMoveModeIndex) (Utilities|Enum|BytetoEnumEFortCustomMovement PreviousCustomMode)) (Utilities|Enum|Equal(Enum) (Variables|Default|GetSwingingGroundedCustomMoveModeIndex) (Utilities|Enum|BytetoEnumEFortCustomMovement PreviousCustomMode))))
        (bind _fort_movement_comp_character (Variables|Default|GetFort%20Movement%20Comp%20Character)
          (:then
            (if (or (Utilities|Enum|NotEqual(Enum) (Variables|/Script/Engine.CharacterMovementComponent|GetMovementMode _fort_movement_comp_character) "MOVE_Custom") (and (Utilities|Enum|NotEqual(Enum) (Variables|Default|GetSwingingCustomMoveModeIndex) (Utilities|Enum|BytetoEnumEFortCustomMovement (Variables|/Script/Engine.CharacterMovementComponent|GetCustomMovementMode _fort_movement_comp_character))) (Utilities|Enum|NotEqual(Enum) (Variables|Default|GetSwingingGroundedCustomMoveModeIndex) (Utilities|Enum|BytetoEnumEFortCustomMovement (Variables|/Script/Engine.CharacterMovementComponent|GetCustomMovementMode _fort_movement_comp_character)))))
              (CallFunction|OnSwingingEnded false)
              (Networking|SwitchHasAuthority
                (:Authority
                  (if (and (Utilities|Enum|Equal(Enum) (Variables|/Script/Engine.CharacterMovementComponent|GetMovementMode _fort_movement_comp_character) "MOVE_Falling") (not (Variables|Status|GetFreeFalling)))
                    (GameplayEffects|ApplyGameplayEffectToSelf :self (Ability|GetFortAbilitySystemComponent (Variables|Default|GetPlayer)) :GameplayEffectClass "/Game/Athena/Items/Parents/FallImmunity/GE_Athena_TillLandFallDamageImmunity_NoFX_Grant_ApplyInAirOnly.GE_Athena_TillLandFallDamageImmunity_NoFX_Grant_ApplyInAirOnly_C" :Level 0.0) ; in case player exited swinging in the air (for some reason)
                  )))))
          (:else))))))

(event Custom|Detach_Client
  (params (DetachFromHip :type Boolean) (DontApplyForce :type Boolean) (ForceRun :type Boolean))
  ;[1 0.5552 0] Detach
  ; (Called from Server - broadcasts to Owning Client)
  (CallFunction|DetachInternal DetachFromHip DontApplyForce ForceRun))

(event Custom|Detach_Internal
  (params (DetachFromHip :type Boolean) (DontApplyForce :type Boolean) (ForceRun :type Boolean))
  (bind _player (Variables|Default|GetPlayer))
  (bind _returnvalue (Pawn|GetController _player))
  (bind _returnvalue_2 (Actor|GetComponentbyClass _returnvalue "/Script/FortniteGame.FortControllerComponent_Swinging"))
  (bind _player_2 (Variables|Default|GetPlayer))
  (bind _returnvalue_3 (Ability|GetFortAbilitySystemComponent _player_2))
  ;[1 0.5552 0] Detach
  ; (Called from Server - broadcasts to Owning Client)
  (Variables|Default|GetPlayer
    (:then
      (Variables|Status|SetbIsPullingActor false)
      (Variables|Default|SetRopeAttached false)
      (Variables|Default|SetbWaitingforBGASpawn false)
      (Variables|Attach|SetLastDetachedTime (Utilities|Time|GetGameTimeinSeconds))
      (if (or (Math|Boolean|ANDBoolean (Default|IsSwingAttached) (or DetachFromHip (not (Variables|/Game/Athena/PlayerPawn_Athena.PlayerPawn_Athena_C|GetSwinging_LockedToHip (Variables|Default|GetPlayer)))) (Variables|Status|GetbIsAttached)) ForceRun)
        (Variables|Status|SetHasDetachedSinceFireWasPressed true)
        (Variables|Status|SetbIsAttached false)
        (Utilities|FlowControl|Sequence
          (:then_0
            ;;[0.4856 0.2926 1] Update if we are attached to an actor that we should ride.
            (CallFunction|UpdatePendingRide))
          (:then_1
            ;;[0.2475 0.7247 1] Stop the player from skidding if they are
            ;; (Server Only)
            (Networking|SwitchHasAuthority
              (:Authority
                (CallFunction|StopSkidding))))
          (:then_2
            ;;[0 0.07145 1] Stop drawing the rope and destroy the attach point actor
            (Rope|DetachRope)
            (bind _attachpointbga (Variables|Attach|GetAttachPointBGA)
              (:then
                (Actor|DestroyActor _attachpointbga)
                (Networking|ForceNetUpdate))
              (:else)))
          (:then_3
            ;;[0 0.3679 1] Tell the player pawn to exit the swing movement mode
            ;; (Server and Owning Client)
            (Utilities|IsValid _returnvalue
              (:Is%20Valid
                (Utilities|IsValid _returnvalue_2
                  (:Is%20Valid
                    (Swinging|SwingDetached _returnvalue_2))))))
          (:then_4
            ;;[0 1 0.5702] Play the swing detach sounds
            ;; (Server Only - multicasts to all clients)
            (Networking|SwitchHasAuthority
              (:Authority
                (if (Variables|Default|GetbPlayDetachAudio)
                  (CallFunction|PlayDetachSoundMulticast)))))
          (:then_5
            ;;[0.3371 1 0.9102] Apply a GE to the player to disallow attaching just after detaching
            ;; (Server Only)
            (Networking|SwitchHasAuthority
              (:Authority
                (Utilities|IsValid _returnvalue_3
                  (:Is%20Valid
                    (GameplayEffects|ApplyGameplayEffectToTarget :self _returnvalue_3 :GameplayEffectClass "/GrappleParent/GrappleParent/GE_GrappleParent_DontUseRightAfterDetach.GE_GrappleParent_DontUseRightAfterDetach_C" :Target _returnvalue_3 :Level 0.0))))))
          (:then_6
            ;;[0 0.6171 1] Launch the player after detaching
            ;; (Owning Client)
            (Networking|SwitchHasAuthority
              (:Remote
                (CallFunction|DetachmentLaunch DontApplyForce))))
          (:then_7
            ;;[0.1939 0.2324 1] Play the launch or detach animation
            ;; (Server and Owning Client)
            (if (not (Variables|Ride|GetPendingRide))
              (bind _returnvalue_4 (Class|FortAnimInputEvent|CreateAnimInputEvent (select (Variables|Status|GetJustLaunched) "Player.Swinging.Launch" "Player.Swinging.Detach")))
              (Animation|TriggerAnimInputEvent (Variables|Default|GetPlayer) _returnvalue_4)))
          (:then_8
            ;;[0.4352 0.191 1] Apply the swing detach force feedback
            ;; (Owning Client Only)
            (Networking|SwitchHasAuthority
              (:Remote
                (ForceFeedback|SpawnForceFeedbackAttached "/GrappleParent/GrappleParent/FF_Detach_GrappleParent.FF_Detach_GrappleParent" (Variables|Visual|GetWeaponMesh)))))
          (:then_9
            (Variables|Status|SetTimeBeenSlow 0)
            (CallFunction|OnDetach)))
        (else
          ;;[0 0.07145 1] Even if the player is not attached, detach the rope and destroy the attach point if it's valid as a failsafe.
          (Rope|DetachRope)
          (bind _attachpointbga_2 (Variables|Attach|GetAttachPointBGA)
            (:then
              (Actor|DestroyActor _attachpointbga_2))
            (:else)))))
    (:else)))

(event Custom|PlayDetachSound_Multicast
  ;[1 0.5552 0] Detach
  ; (Called from Server - broadcasts to Owning Client)
  ;;[0 1 0.5702] Play the swing detach sounds
  ;; (Server Only - multicasts to all clients)
  (Networking|SwitchHasAuthority
    (:Remote
      (bind _returnvalue (Audio|SpawnSoundAttached (Variables|Sound|GetDetachSound) (Variables|Visual|GetWeaponMesh)))
      (CallFunction|FillAudioComponentParams _returnvalue))))

(event Custom|Multicast_PlayPickupEffects
  (params (Location :type Vector) (PickupActor :type Actor%20Object%20Reference))
  ;[1 0.523 0.9872] Helper function to play the pickup effects on clients from the server.
  (Networking|CanExecuteCosmeticEvents
    (:True
      (Audio|PlaySoundatLocation (Variables|Sound|GetPickupSound) Location)
      (GameplayCue|ExecuteGameplayCue%28Burst%29Local PickupActor (Variables|Default|GetPickup%20Impact%20FX) (Ability|GameplayCue|MakeGameplayCueParameters 0.0)))))

(event Custom|OnDetach)

(event Custom|UpdateAnimLayer
  (params (ShouldBeInLayer :type Boolean))
  (bind _player (Variables|Default|GetPlayer))
  (bind _swinginganimlayerclass (Variables|Default|GetSwingingAnimLayerClass))
  ;[0 1 0.9484] Slower Update Timer
  ; (Server and All Clients)
  ;;[1 0.1676 0.7849] Failsafe to make sure animation and camera effects properly turn on/off based on if player is swinging or not
  (if (Pawn|IsLocallyControlled (Variables|Default|GetPlayer))
    (CallFunction|UpdateAnimationStateServer ShouldBeInLayer))
  (if ShouldBeInLayer
    (Animation|LinkAnimLayer _player _swinginganimlayerclass)
    (else
      (Animation|UnlinkAnimLayer _player _swinginganimlayerclass))))

(event Custom|UpdateAnimationStateServer
  (params (ShouldBeInLayer :type Boolean))
  (bind _swinginganimlayerclass (Variables|Default|GetSwingingAnimLayerClass))
  (bind _player (Variables|Default|GetPlayer))
  ;[0 1 0.9484] Slower Update Timer
  ; (Server and All Clients)
  ;;[1 0.1676 0.7849] Failsafe to make sure animation and camera effects properly turn on/off based on if player is swinging or not
  (Networking|SwitchHasAuthority
    (:Authority
      (if ShouldBeInLayer
        (Animation|MulticastLinkAnimLayer _player _swinginganimlayerclass)
        (else
          (Animation|MulticastUnlinkAnimLayer _player _swinginganimlayerclass)
          (CallFunction|ResetMeshTransform))))))

(event Custom|DonePullingPickup
  ;[0.009286 0.2635 0.04312] Spawn Attachment
  (CallFunction|MulticastSetBuildModeBlocked false)
  (if (Variables|Status|GetbIsPullingActor)
    (Rope|DetachRope)
    (Variables|Status|SetbIsPullingActor false)))

(event Custom|SpawnAttachment
  (bind (_bblockinghit _binitialoverlap _time _distance _location _impactpoint _normal) (Collision|BreakHitResult (Variables|Targeting|GetTargetingHitResult)))
  (bind (_origin _boxextent) (Collision|GetActorBounds (Variables|Targeting|GetActorAttachedTo) true false))
  (bind (_boxextent_x _boxextent_y _boxextent_z) (Math|Vector|BreakVector _boxextent))
  (bind _actorattachedto (Variables|Targeting|GetActorAttachedTo))
  (bind _returnvalue (Transformation|GetActorLocation _actorattachedto))
  (bind _targetinghitresult (Variables|Targeting|GetTargetingHitResult))
  (bind (_bblockinghit_2 _binitialoverlap_2 _time_2 _distance_2 _location_2 _impactpoint_2 _normal_2 _impactnormal _physmat _hitactor) (Collision|BreakHitResult _targetinghitresult))
  (bind _actorattachedto_2 (Variables|Targeting|GetActorAttachedTo))
  (bind (_asfort_vehicle_interface _bsuccess) (Utilities|Casting|CastToFortVehicleInterface _actorattachedto_2))
  (bind _actorattachedto_3 (Variables|Targeting|GetActorAttachedTo))
  ;[0.009286 0.2635 0.04312] Spawn Attachment
  (Networking|SwitchHasAuthority
    (:Authority
      (Variables|Default|GetPlayer
        (:then
          (Utilities|IsValid _hitactor
            (:Is%20Valid
              (bind _output_get (Variables|Targeting|SetActorAttachedTo _hitactor))
              (Variables|Targeting|SetActorAttachedTo_OriginalPosition (Transformation|GetActorLocation _hitactor))
              (CallFunction|QuestAttachedtoTree _output_get)
              (if (Class|FortWeaponComponentSwinging|IsActorPullable (Variables|Default|GetGrappleSwingingComp) _hitactor)
                (bind _returnvalue_2 (Projectile|SpawnProjectile :ProjectileClass "/GrappleParent/GrappleParent/Prj_GrappleParent_PickUpTaker.Prj_GrappleParent_PickUpTaker_C" :RequestedBy (Variables|Default|GetPlayer) :SpawnLocation _returnvalue :SpawnRotation "0, 0, 0" :SpawnDirection (Math|Conversions|RotationFromXVector (Math|Vector|Normalize (- (Transformation|GetActorLocation (Variables|Default|GetPlayer)) _returnvalue))) :bUseDefaultPhysics false :InitialSpeed 0.0 :GravityScale 0.0 :ChargePercent 0.0))
                (bind _asprj_grapple_parent_pick_up_taker (Utilities|Casting|CastToPrj_GrappleParent_PickUpTaker _returnvalue_2))
                (Rope|AttachRope _actorattachedto_3)
                (Variables|Status|SetbIsPullingActor true)
                (Class|PrjGrappleParentPickUpTaker|Init _asprj_grapple_parent_pick_up_taker _actorattachedto_3)
                (CallFunction|MulticastPlayPickupEffects _returnvalue _actorattachedto_3)
                (Default|BindEventtoOnDonePullingPickup _asprj_grapple_parent_pick_up_taker (delegate Custom|DonePullingPickup))
                (else
                  (Animation|MulticastLinkAnimLayer (Variables|Default|GetPlayer) (Variables|Default|GetSwingingAnimLayerClass))
                  (if (and (CallFunction|PlayerInputWantsAttachment) (not (GameplayTags|HasAnyMatchingGameplayTags (Variables|Default|GetPlayer) (GameplayTags|MakeLiteralGameplayTagContainer "(GameplayTags=((TagName=\"Granted.Athena.EnvItem.HidingProp.GettingInProp\")))"))))
                    (NetTolerance|SetUseHighAccelerationTolerance (Variables|Default|GetFort%20Movement%20Comp%20Character) true "GrappleGloves")
                    (GameplayEffects|RemoveActiveEffectswithGrantedTags (Ability|GetFortAbilitySystemComponent (Variables|Default|GetPlayer)) (Variables|Default|GetBlockUseTag))
                    (bind _returnvalue_3 (Building|SpawnBuildingActor (Variables|Attach|GetAttachPointClass) (Math|Transform|MakeTransform _impactpoint (Math|Rotator|MakeRotfromX (Math|Vector|NegateVector _normal))) 0 (Variables|Default|GetPlayer) false false))
                    (bind _shouldattach (CallFunction|CheckForVehicle))
                    (if _shouldattach
                      (if (< (Utilities|Operators|Add _boxextent_x _boxextent_y _boxextent_z) 1e+04)
                        (if (and (Class|AthenaFunctionLibrary|RowtoBool (Variables|Default|GetSF_ShouldAttachBGAToVehicles)) _bsuccess)
                          (Transformation|AttachActorToActor _returnvalue_3 (Variables|Targeting|GetActorAttachedTo) (CallFunction|GetVehicleAttachSocket _asfort_vehicle_interface) "SnapToTarget" "KeepWorld" "KeepWorld" false)
                          (else
                            (Transformation|AttachActorToActor _returnvalue_3 (Variables|Targeting|GetActorAttachedTo) "None" "KeepWorld" "KeepWorld" "KeepWorld" false)))))
                    (else
                      (CallFunction|SpawnAttachmentFailed))))))))
        (:else
          (CallFunction|SpawnAttachmentFailed))))))

(event Custom|Spawn%20Attachment%20Failed
  ;[0.009286 0.2635 0.04312] Spawn Attachment
  ;;[1 0.4636 0] Spawn Attachment Failed.
  ;; Allow the player to enter build mode again.
  (Networking|SwitchHasAuthority
    (:Authority
      (CallFunction|MulticastSetBuildModeBlocked false))))

(event Custom|Fall
  ;[1 0.5037 0.1334] Fall
  (Networking|SwitchHasAuthority
    (:Authority
      (if (and (CallFunction|IsSwinging) (Variables|Default|GetFallWhenShot))
        (Variables|Ride|SetbBlockRiding true)
        (Variables|Status|SetFreeFalling true)
        (CallFunction|Detach true true)
        (Development|PrintString "NYXGLASS_DETACH_fall")
        (CallFunction|FallClient)
        (CallFunction|ApplyCooldown true)
        (bind _swingingcontroller (Variables|Default|GetSwingingController)
          (:then
            (Swinging|EndSwingingMovement _swingingcontroller))
          (:else
            (CallFunction|OnSwingingEnded false)))))))

(event Custom|QuestAttachedToTree
  (params (HitActor :type Actor%20Object%20Reference))
  ;[0.009286 0.2635 0.04312] Spawn Attachment
  ;;[0.0185 1 0.3916] If attached to tree - grant a tag to the player to be identified for a quest
  (Networking|SwitchHasAuthority
    (:Authority
      (Ability|SendGameplayEventtoActor (Variables|Default|GetPlayer) "(TagName=\"Event.Quests.SwingingHit\")" (Utilities|Struct|MakeGameplayEventData "(TagName=\"Event.Quests.SwingingHit\")" (Variables|Default|GetPlayer) HitActor))
      (bind _returnvalue (GameplayTags|HasMatchingGameplayTag(Message) HitActor "(TagName=\"Building.Type.Prop.Athena.Quest.Tree\")"))
      (if _returnvalue
        (GameplayEffects|ApplyGameplayEffectToTarget :self (Ability|GetFortAbilitySystemComponent (Variables|Default|GetPlayer)) :GameplayEffectClass "/GrappleParent/GrappleParent/GE_GrappleParent_AttachedToTree_Quest.GE_GrappleParent_AttachedToTree_Quest_C" :Target (Ability|GetFortAbilitySystemComponent (Variables|Default|GetPlayer)) :Level 0.0)))))

(event Custom|PlayerHit
  (params (SelfActor :type Actor%20Object%20Reference) (OtherActor :type Actor%20Object%20Reference) (NormalImpulse :type Vector) (Hit :type Hit%20Result%20Structure%20%28by%20ref%29))
  ;[1 0.5037 0.1334] Fall
  (Variables|Status|SetFreeFalling false))

(event Custom|Fall_Client
  ;[1 0.5037 0.1334] Fall
  (ForceFeedback|SpawnForceFeedbackAttached "/GrappleParent/GrappleParent/FF_GotShot_GrappleParent.FF_GotShot_GrappleParent" (Variables|Visual|GetWeaponMesh))
  (Variables|Status|SetFreeFalling true)
  (Utilities|ShowTip "(TagName=\"Athena.Ability.Failed.WestSausage.WebGotShot\")")
  (bind _swingingcontroller (Variables|Default|GetSwingingController)
    (:then
      (Swinging|EndSwingingMovement _swingingcontroller))
    (:else
      (CallFunction|OnSwingingEnded false))))

(event Custom|SetBuildModeBlocked
  (params (bIsBlocked :type Boolean))
  (bind _blockbuildmodetags (Variables|Default|GetBlockBuildModeTags))
  ;[1 0.3804 0.6592] Set build mode blocked while swinging. Ensure this item only ever applies and removes one set of build mode blocking tags.
  (if (ScalableFloat|GetValueasBool (Variables|Default|GetSF_EnableBuildModeBlocking))
    (bind _player (Variables|Default|GetPlayer)
      (:then
        (if bIsBlocked
          (if (not (and (or (FortWeapon|IsEquipped) (not (ScalableFloat|GetValueasBool (Variables|Default|GetSF_BlockBuildModeOnlyWhenEquipped)))) (Variables|Default|GetbIsBuildModeBlocked)))
            (Ability|GameplayEffect|AddLooseGameplayTags _player _blockbuildmodetags)
            (Variables|Default|SetbIsBuildModeBlocked true))
          (elif (Variables|Default|GetbIsBuildModeBlocked)
            (Ability|GameplayEffect|RemoveLooseGameplayTags _player _blockbuildmodetags)
            (Variables|Default|SetbIsBuildModeBlocked false))))
      (:else))))

(event Custom|CustomFire
  (params (TargetingHitResult :type Hit%20Result%20Structure))
  (bind _projectilespawndelay (Variables|Default|GetProjectileSpawnDelay))
  (bind _returnvalue (ScalableFloat|GetValueatLevel _projectilespawndelay))
  ;[0.009286 0.2635 0.04312] Spawn Attachment
  ;;[0.01873 1 0.391] Custom Fire
  (Variables|Targeting|SetTargetingHitResult TargetingHitResult)
  (CallFunction|SetJustFired)
  (if (> _returnvalue 0)
    (Utilities|FlowControl|Delay _returnvalue))
  (Variables|Default|SetbWaitingforBGASpawn true)
  (CallFunction|SpawnAttachment)
  (CallFunction|ReplicatedFireSound)
  (CallFunction|SetBuildModeBlocked true) ; Sets build mode blocked on the server. \nWill be unblocked in Detach.
)

(event Custom|ReplicatedFireSound
  ;[0.009286 0.2635 0.04312] Spawn Attachment
  (Networking|CanExecuteCosmeticEvents
    (:True
      (Audio|SpawnSoundAttached (Variables|Sound|GetFireSound) (Variables|Visual|GetWeaponMesh)))))

(event Custom|PlaySwingSound
  ;[0.009286 0.2635 0.04312] Spawn Attachment
  (Networking|SwitchHasAuthority
    (:Remote
      (Audio|SpawnSoundAttached (Variables|Sound|GetSwingSound) (Variables|Visual|GetWeaponMesh)))))

(event Custom|Multicast_SetBuildModeBlocked
  (params (bIsBlocked :type Boolean))
  ;[1 0.3804 0.6592] Set build mode blocked while swinging. Ensure this item only ever applies and removes one set of build mode blocking tags.
  (CallFunction|SetBuildModeBlocked bIsBlocked))

(event EventEndPlay
  (params (EndPlayReason :type EEndPlayReason%20Enum :default "Destroyed"))
  ;[1 0.3804 0.6592] Set build mode blocked while swinging. Ensure this item only ever applies and removes one set of build mode blocking tags.
  (|CallParentFunction EndPlayReason)
  (CallFunction|SetBuildModeBlocked false))

(event Custom|TryToFire
  ;[0.009286 0.2635 0.04312] Spawn Attachment
  ;;[0.01873 1 0.391] Custom Fire
  (if (Utilities|CanAttach)
    (Variables|Attach|SetLastAttachedTime (Utilities|Time|GetGameTimeinSeconds))
    (CallFunction|SetJustFired)
    (CallFunction|CustomFire (CallFunction|GetTargetingResult true))
    (FortWeapon|PlayForceFeedback false)
    (CallFunction|SetBuildModeBlocked true) ; Sets build mode blocked on the client. \nWill be unblocked in Detach.
  ))

(event Custom|FirePressed
  (bind (_blockedbytag _blockedbynoammo) (CallFunction|IsFiringBlocked))
  (if _blockedbytag
    (FailedAction|UnabletoPerformAction (Variables|Default|GetPlayer) false (GameplayTags|MakeLiteralGameplayTagContainer "(GameplayTags=((TagName=\"Athena.Ability.Failed.Failed.GenericFail\")))")) ; show reason for failing on screen if one is given
    (elif _blockedbynoammo
      (FailedAction|UnabletoPerformAction (Variables|Default|GetPlayer) false (GameplayTags|MakeLiteralGameplayTagContainer "(GameplayTags=((TagName=\"Athena.Ability.Failed.OutOfAmmo\")))")) ; show reason for failing on screen if one is given
      (else
        (Variables|Status|SetHoldingFire true)
        (GameplayEffects|ApplyGameplayEffectToTarget :self (Ability|GetFortAbilitySystemComponent (Variables|Default|GetPlayer)) :GameplayEffectClass "/GrappleParent/GrappleParent/GE_GrappleParent_HoldingFireButton.GE_GrappleParent_HoldingFireButton_C" :Target (Ability|GetFortAbilitySystemComponent (Variables|Default|GetPlayer)) :Level 0.0) ; GE to signify player is holding fire button
        (Networking|SwitchHasAuthority
          (:Remote
            (if (or (Variables|Targeting|GetTargetingSomething) (Variables|Default|GetHasCustomTargeting))
              (CallFunction|TrytoFire))))))))

(event Custom|SetJustFired
  ;[0.009286 0.2635 0.04312] Spawn Attachment
  (Variables|Default|SetbJustFired true)
  (Utilities|Time|SetTimerbyEvent (delegate Custom|JustFiredTimerElapsed) 0.2))

(event Custom|JustFiredTimerElapsed
  ;[0.009286 0.2635 0.04312] Spawn Attachment
  (Variables|Default|SetbJustFired false))

(event Custom|FireReleased
  (Variables|Status|SetHoldingFire false)
  (Networking|SwitchHasAuthority
    (:Authority
      (Development|PrintString "NYXGLASS_NormalDet_server"))
    (:Remote
      (CallFunction|RequestDetachfromClient false)
      (Development|PrintString "NYXGLASS_NormalDet_client"))))

(fn AllowStopSwinging
  (return true))

(fn CalculateChargeTime
  (return 0.0))

(fn CalculateCooldownTime
  (return (Utilities|RemapFloat (- (Fortnite|Sync|Time|GetServerWorldTimeSeconds) (Variables|Status|GetStartedSwingingTime)) (Math|Vector2D|MakeVector2D 0.0 (Utilities|GetValueatZero (Variables|Design|GetCooldown_TimeForNormalDuration))) (Math|Vector2D|MakeVector2D (Utilities|GetValueatZero (Variables|Design|GetCooldown_MinimumDuration)) (Utilities|GetValueatZero (Variables|Design|GetCooldown_NormalDuration))) true)))

(fn CalculateDetachmentForce
  (params (VectorToAttachPoint :type Vector) (Launch_ :type Boolean) (LaunchVector_ :type Vector))
  (bind _returnvalue (Math|Vector|VectorLength VectorToAttachPoint))
  (LocalVariables|SetLaunchVector LaunchVector_)
  (LocalVariables|SetLaunch Launch_)
  (LocalVariables|SetVecToAttachPoint_Length _returnvalue)
  (LocalVariables|SetVecToAttachPoint_Dir (/ VectorToAttachPoint _returnvalue))
  (bind (_grounded _hitnormal _distance) (Class|PlayerPawnAthena|SwingingCheckGrounded (Variables|Default|GetPlayer)))
  (bind (_returnvalue_x _returnvalue_y _returnvalue_z) (Math|Vector|BreakVector (Utilities|Operators|Add (Variables|/Script/Engine.MovementComponent|GetVelocity (Variables|/Script/Engine.Character|GetCharacterMovement (Variables|Default|GetPlayer))) "0,0,200.000000" (Utilities|Operators|Multiply (Utilities|RemapFloat _distance "(X=100.000000,Y=600.000000)" (select (Utilities|IsValid (Actor|GetAttachParentActor (Variables|/Game/Athena/PlayerPawn_Athena.PlayerPawn_Athena_C|GetSwinging_AttachBGA (Variables|Default|GetPlayer))) :!class /Script/Engine.KismetSystemLibrary) "(X=0.800000,Y=1.000000)" "(X=0.500000,Y=0.600000)") true) (Utilities|RemapFloat (Utilities|SqrtNegative (Math|Vector|DotProduct (* (LocalVariables|GetVecToAttachPoint_Dir) (Math|Float|Max(Float) (LocalVariables|GetVecToAttachPoint_Length) 2.5e+02)) (Variables|/Script/Engine.MovementComponent|GetVelocity (Variables|/Script/Engine.Character|GetCharacterMovement (Variables|Default|GetPlayer))))) "(X=-200.000000,Y=900.000000)" "(X=0.000000,Y=1.800000)" true) (* (LocalVariables|GetVecToAttachPoint_Dir) (Math|Float|Clamp(Float) (LocalVariables|GetVecToAttachPoint_Length) 3e+02 4.5e+02))))))
  (return (Math|Vector|SelectVector (* (Math|Vector|Normalize (LocalVariables|GetLaunchVector)) (Math|Float|Clamp(Float) (Math|Vector|VectorLength (Variables|/Script/Engine.MovementComponent|GetVelocity (Variables|/Script/Engine.Character|GetCharacterMovement (Variables|Default|GetPlayer)))) (ScalableFloat|GetValueatLevel (LocalVariables|GetLaunchOverMinForce)) (ScalableFloat|GetValueatLevel (LocalVariables|GetLaunchOverMaxForce)))) (* (Math|Vector|MakeVector _returnvalue_x _returnvalue_y (Math|Float|Max(Float) _returnvalue_z (* (Utilities|RemapFloat _distance "(X=100.000000,Y=600.000000)" (select (Utilities|IsValid (Actor|GetAttachParentActor (Variables|/Game/Athena/PlayerPawn_Athena.PlayerPawn_Athena_C|GetSwinging_AttachBGA (Variables|Default|GetPlayer))) :!class /Script/Engine.KismetSystemLibrary) "(X=0.800000,Y=1.000000)" "(X=0.500000,Y=0.600000)") true) (ScalableFloat|GetValueatLevel (LocalVariables|GetMinZForceMultiplier))))) (ScalableFloat|GetValueatLevel (LocalVariables|GetOverallMultiplier))) (and (> (.z (* (Math|Vector|Normalize (LocalVariables|GetLaunchVector)) (Math|Float|Clamp(Float) (Math|Vector|VectorLength (Variables|/Script/Engine.MovementComponent|GetVelocity (Variables|/Script/Engine.Character|GetCharacterMovement (Variables|Default|GetPlayer)))) (ScalableFloat|GetValueatLevel (LocalVariables|GetLaunchOverMinForce)) (ScalableFloat|GetValueatLevel (LocalVariables|GetLaunchOverMaxForce))))) (.z (* (Math|Vector|MakeVector _returnvalue_x _returnvalue_y (Math|Float|Max(Float) _returnvalue_z (* (Utilities|RemapFloat _distance "(X=100.000000,Y=600.000000)" (select (Utilities|IsValid (Actor|GetAttachParentActor (Variables|/Game/Athena/PlayerPawn_Athena.PlayerPawn_Athena_C|GetSwinging_AttachBGA (Variables|Default|GetPlayer))) :!class /Script/Engine.KismetSystemLibrary) "(X=0.800000,Y=1.000000)" "(X=0.500000,Y=0.600000)") true) (ScalableFloat|GetValueatLevel (LocalVariables|GetMinZForceMultiplier))))) (ScalableFloat|GetValueatLevel (LocalVariables|GetOverallMultiplier))))) (LocalVariables|GetLaunch)))))

(fn CheckforVehicle
  (if (Class|AthenaFunctionLibrary|RowtoBool (Variables|Default|GetSF_ShouldAttachBGAToVehicles))
    (return true)
    (else
      (Utilities|Casting|CastToFortVehicleInterface (Variables|Targeting|GetActorAttachedTo)
        (:then
          (return false))
        (:CastFailed
          (if (Class|AthenaFunctionLibrary|RowtoBool (Variables|Default|GetSF_ShouldAttachBGAToPassengers))
            (return true)
            (else
              (bind _asfort_player_pawn (Utilities|Casting|CastToFortPlayerPawn (Variables|Targeting|GetActorAttachedTo))
                (:then
                  (if (Vehicle|IsinVehicle _asfort_player_pawn)
                    (return false)
                    (else
                      (return true))))
                (:CastFailed
                  (return true))))))))))

(fn CheckAttachmentWhenUnequipped
  (if (and (not (FortWeapon|IsEquipped)) (Variables|Status|GetbIsAttached))
    (CallFunction|Detach true true)))

(fn CheckIfShouldLaunch
  (bind _player (Variables|Default|GetPlayer))
  (bind _returnvalue (Fortnite|GetFortPlayerControllerfromActor _player))
  (bind _player_2 (Variables|Default|GetPlayer))
  (bind _returnvalue_2 (Transformation|GetActorLocation _player_2))
  (bind (_highestvector_x _highestvector_y _highestvector_z) (Math|Vector|BreakVector (Variables|GetHighestVector_X)))
  (Utilities|IsValid _returnvalue
    (:Is%20Valid
      (bind (_outhit _returnvalue_3) (Collision|SphereTraceByProfile (Utilities|Operators|Add (Transformation|GetActorLocation (Variables|Default|GetPlayer)) "0,0,20.000000" (* (Math|Vector|Normalize2D(Vector) (Math|Vector|SelectVector (Math|Vector|GetForwardVector (Camera|GetCameraRotation (Variables|/Script/Engine.PlayerController|GetPlayerCameraManager _returnvalue))) (Math|Vector|MakeVector (.x (Variables|/Script/Engine.MovementComponent|GetVelocity (Variables|/Script/Engine.Character|GetCharacterMovement (Variables|Default|GetPlayer)))) (.y (Variables|/Script/Engine.MovementComponent|GetVelocity (Variables|/Script/Engine.Character|GetCharacterMovement (Variables|Default|GetPlayer)))) (Math|Float|Max(Float) (.z (Variables|/Script/Engine.MovementComponent|GetVelocity (Variables|/Script/Engine.Character|GetCharacterMovement (Variables|Default|GetPlayer)))))) (< (Math|Vector|VectorLengthSquared (Variables|/Script/Engine.MovementComponent|GetVelocity (Variables|/Script/Engine.Character|GetCharacterMovement (Variables|Default|GetPlayer)))) 5e+02)) 0.01) -1e+02)) (+ (Utilities|Operators|Add (Transformation|GetActorLocation (Variables|Default|GetPlayer)) "0,0,20.000000" (* (Math|Vector|Normalize2D(Vector) (Math|Vector|SelectVector (Math|Vector|GetForwardVector (Camera|GetCameraRotation (Variables|/Script/Engine.PlayerController|GetPlayerCameraManager _returnvalue))) (Math|Vector|MakeVector (.x (Variables|/Script/Engine.MovementComponent|GetVelocity (Variables|/Script/Engine.Character|GetCharacterMovement (Variables|Default|GetPlayer)))) (.y (Variables|/Script/Engine.MovementComponent|GetVelocity (Variables|/Script/Engine.Character|GetCharacterMovement (Variables|Default|GetPlayer)))) (Math|Float|Max(Float) (.z (Variables|/Script/Engine.MovementComponent|GetVelocity (Variables|/Script/Engine.Character|GetCharacterMovement (Variables|Default|GetPlayer)))))) (< (Math|Vector|VectorLengthSquared (Variables|/Script/Engine.MovementComponent|GetVelocity (Variables|/Script/Engine.Character|GetCharacterMovement (Variables|Default|GetPlayer)))) 5e+02)) 0.01) -1e+02)) (* (Math|Vector|Normalize2D(Vector) (Math|Vector|SelectVector (Math|Vector|GetForwardVector (Camera|GetCameraRotation (Variables|/Script/Engine.PlayerController|GetPlayerCameraManager _returnvalue))) (Math|Vector|MakeVector (.x (Variables|/Script/Engine.MovementComponent|GetVelocity (Variables|/Script/Engine.Character|GetCharacterMovement (Variables|Default|GetPlayer)))) (.y (Variables|/Script/Engine.MovementComponent|GetVelocity (Variables|/Script/Engine.Character|GetCharacterMovement (Variables|Default|GetPlayer)))) (Math|Float|Max(Float) (.z (Variables|/Script/Engine.MovementComponent|GetVelocity (Variables|/Script/Engine.Character|GetCharacterMovement (Variables|Default|GetPlayer)))))) (< (Math|Vector|VectorLengthSquared (Variables|/Script/Engine.MovementComponent|GetVelocity (Variables|/Script/Engine.Character|GetCharacterMovement (Variables|Default|GetPlayer)))) 5e+02)) 0.01) 1.5e+03)) 5e+01 "FortProjectileHitAllPawns" false (Utilities|Array|MakeArray (Variables|Default|GetPlayer)) "None" true "(R=1.000000,G=0.000000,B=0.000000,A=1.000000)" "(R=0.000000,G=1.000000,B=0.000000,A=1.000000)" 1e+01))
      (if _returnvalue_3
        (LocalVariables|SetDistance (.Distance (Collision|BreakHitResult _outhit)))
        (LocalVariables|SetLocation (.ImpactPoint (Collision|BreakHitResult _outhit)))
        (LocalVariables|SetNormal (.Normal (Collision|BreakHitResult _outhit)))
        (LocalVariables|SetLocation_VerticalSnapped (+ (Utilities|Operators|Subtract (LocalVariables|GetLocation) (Math|Vector|MakeVector 0.0 0.0 (Math|Float|%25%28Float%29 (.z (LocalVariables|GetLocation)) (LocalVariables|GetTileHeight)))) (* (LocalVariables|GetNormal) 2.5e+02)))
        (Variables|SetOneTileUp_X (Math|Vector|MakeVector 0.0 0.0 (LocalVariables|GetTileHeight)))
        (for _index (range 1 3)
          (bind (_outhit_2 _returnvalue_4) (Collision|SphereTraceByProfile (+ (LocalVariables|GetLocation_VerticalSnapped) (* (LocalVariables|GetOneTileUp) _index)) (+ (+ (LocalVariables|GetLocation_VerticalSnapped) (* (LocalVariables|GetOneTileUp) _index)) (* (LocalVariables|GetNormal) -6e+02)) 192.0 "FortProjectileHitAllPawns" false (Utilities|Array|MakeArray (Variables|Default|GetPlayer)) "None" true "(R=1.000000,G=0.000000,B=0.000000,A=1.000000)" "(R=0.000000,G=1.000000,B=0.000000,A=1.000000)" 2e+01))
          (if _returnvalue_4
            (if (< _highestvector_z (.z (Math|Vector|Normalize (- (.ImpactPoint (Collision|BreakHitResult _outhit_2)) _returnvalue_2) 0.01)))
              (LocalVariables|SetHighestLocation (.ImpactPoint (Collision|BreakHitResult _outhit_2)))
              (LocalVariables|SetHitSomethingAbove true)
              (LocalVariables|SetHighestVector (Math|Vector|Normalize (- (.ImpactPoint (Collision|BreakHitResult _outhit_2)) _returnvalue_2) 0.01)))))
        (return true (Math|Vector|SelectVector (+ (* (Math|Vector|Normalize2D(Vector) (LocalVariables|GetNormal) 0.01) (+ (* (.z (+ (* (- (Math|Vector|SelectVector (LocalVariables|GetHighestLocation) (LocalVariables|GetLocation) (LocalVariables|GetHitSomethingAbove)) (Transformation|GetActorLocation (Variables|Default|GetPlayer))) "1.000000,1.000000,2.500000") "0,0,300.000000")) 0.15) 3e+02)) (+ (* (- (Math|Vector|SelectVector (LocalVariables|GetHighestLocation) (LocalVariables|GetLocation) (LocalVariables|GetHitSomethingAbove)) (Transformation|GetActorLocation (Variables|Default|GetPlayer))) "1.000000,1.000000,2.500000") "0,0,300.000000")) (+ (* (- (Math|Vector|SelectVector (LocalVariables|GetHighestLocation) (LocalVariables|GetLocation) (LocalVariables|GetHitSomethingAbove)) (Transformation|GetActorLocation (Variables|Default|GetPlayer))) "1.000000,1.000000,2.500000") "0,0,300.000000") (< (LocalVariables|GetDistance) 1.5e+02)))
        (else
          (return false "0.000000, 0, 0"))))))

(fn DelayBetweenRopeAttachAndMovement
  (return false 0.0))

(fn DetachmentLaunch
  (params (DontApplyForce :type Boolean))
  (if (and (not DontApplyForce) (not (Variables|Ride|GetPendingRide)))
    (bind (_shouldlaunch _launchvector) (CallFunction|CheckifShouldLaunch)) ; Does 3 spherecasts on launch, 1 spherecast if failed to launch
    (Variables|Status|SetJustLaunched _shouldlaunch)
    (CallFunction|JumpAndImpulse(Override) (CallFunction|CalculateDetachmentForce (Swinging|SwingingVectortoAttachPoint (Variables|Default|GetPlayer) "weapon calc detach force") _shouldlaunch _launchvector))))

(fn FillAudioComponentParams
  (params (Audio%20Component :type Audio%20Component%20Object%20Reference))
  (Audio|Parameter|SetBooleanParameter Audio%20Component "IsLocal" (Pawn|IsLocallyViewed (Variables|Default|GetPlayer)) :!class /Script/AudioExtensions.AudioParameterControllerInterface)
  (Audio|Parameter|SetBooleanParameter Audio%20Component "IsTeam" (Utilities|Enum|Equal(Enum) (Team|GetTeamAffiliationforActortoLocalPlayer (Variables|Default|GetPlayer)) "Friendly") :!class /Script/AudioExtensions.AudioParameterControllerInterface))

(fn GetAffiliationsToShowFor
  (return (Utilities|Array|MakeArray "Neutral" "Hostile")))

(fn GetTargetingResult
  (params (GettingFireTarget :type Boolean))
  (return (Variables|/Script/FortniteGame.FortWeaponComponent_Swinging|GetTargetingHitResult (Variables|Default|GetGrappleSwingingComp))))

(fn GetTrueVelocityBasedOnTransform
  (bind _player (Variables|Default|GetPlayer))
  (bind _returnvalue (FortMovement|GetPreviousPosition _player))
  (return (/ (- _returnvalue (Transformation|GetActorLocation _player)) (Utilities|Time|GetWorldDeltaSeconds))))

(fn GetVehicleAttachSocket
  (params (Vehicle :type Fort%20Vehicle%20Interface%20Interface))
  (bind _returnvalue (Class|FortVehicleInterface|GetSkeletalMeshComponent Vehicle))
  ;[0.6354 0.5889 0 0.5] Give the vehicle socket that this should attach to, if the socket does not exist this will return None to attach to the root of the vehicle FORT-885282
  (if (Utilities|IsValid _returnvalue :!class /Script/Engine.KismetSystemLibrary)
    (return (select (Transformation|DoesSocketExist _returnvalue (Variables|Default|GetVehicle%20Attach%20Socket)) (Variables|Default|GetVehicle%20Attach%20Socket) "None"))
    (else
      (return "None"))))

(fn InitialJumpMultiplier
  (return 1.0))

(fn IsFiringBlocked
  (bind _blockingtags (Variables|Default|GetBlockingTags))
  (bind _player (Variables|Default|GetPlayer))
  (bind _returnvalue (GameplayTags|HasAnyMatchingGameplayTags _player _blockingtags))
  (bind _returnvalue_2 (FortWeapon|GetMagazineAmmoCount))
  (bind _player_2 (Variables|Default|GetPlayer))
  (bind _returnvalue_3 (GameplayTags|HasMatchingGameplayTag _player_2 "(TagName=\"MovementMode.Grinding\")"))
  (bind _sf_blockattachmentwhilegrinding (Variables|Default|GetSF_BlockAttachmentWhileGrinding))
  (bind _returnvalue_4 (and _returnvalue_3 _sf_blockattachmentwhilegrinding))
  (bind _returnvalue_5 (<= _returnvalue_2 0))
  (bind _experiencesettingscomp (Variables|Default|GetExperienceSettingsComp))
  (if (Utilities|IsValid _experiencesettingscomp :!class /Script/Engine.KismetSystemLibrary)
    (return (Math|Boolean|ORBoolean _returnvalue _returnvalue_4 (and (not (ScalableFloat|GetValueasBoolwithContextParam (Variables|Default|GetPlayer) (Variables|/Script/FortniteGame.FortExperienceSettingsComponent|GetNoCooldowns _experiencesettingscomp))) (GameplayTags|HasAnyMatchingGameplayTags (Variables|Default|GetPlayer) (Variables|Default|GetCooldownTags)))) _returnvalue_5)
    (else
      (return (or _returnvalue _returnvalue_4) _returnvalue_5))))

(fn IsSwingAttached
  (bind _player (Variables|Default|GetPlayer)
    (:then
      (return (or (Swinging|GetIsSwingAttached _player) (Variables|Status|GetbIsAttached))))
    (:else
      (return false))))

(fn IsSwinging
  (bind _player (Variables|Default|GetPlayer)
    (:then
      (return (or (Swinging|GetIsSwinging _player) (Variables|Status|GetbIsAttached))))
    (:else
      (return false))))

(fn JumpandImpulse(Override)
  (params (New%20Velocity :type Vector))
  (Swinging|JumpandImpulse (Variables|Default|GetSwingingController) (- New%20Velocity (Math|Conversions|ToVector2D(Vector) (Variables|/Script/Engine.MovementComponent|GetVelocity (Variables|/Script/Engine.Character|GetCharacterMovement (Variables|Default|GetPlayer)))))))

(fn PlayerInputWantsAttachment
  (return (or (Variables|Status|GetHoldingFire) (and (Variables|Default|GetTapToAttachMode) (or (Variables|Status|GetbIsAttached) (Variables|Default|GetbWaitingforBGASpawn))))))

(fn ResetMeshTransform)

(fn ShouldBlockDetachment
  (return false))

(fn ShouldBlockSwingEndOnLand
  (return false))

(fn ShouldDetachIfTooSlow
  (return true))

(fn ShouldDisableFiring
  (return false))

(fn ConstructionScript
  ;[1 1 1] Set the Grapple Swinging Component variable. It's odd, but we do this because the GrappleSwingingComp variable was already in use, and this ensures we don't break anything else that references it.
  (Variables|Default|SetGrappleSwingingComp (Variables|Default|GetGrappleSwingingComponent))
  (if (Variables|Default|GetTapToAttachMode)
    (Variables|Attach|SetbRequireHeldInputForAttach false)))

(fn VectorToAttachPoint
  (bind _player (Variables|Default|GetPlayer))
  (Utilities|IsValid (Variables|Attach|GetAttachPointBGA)
    (:Is%20Valid
      (Utilities|IsValid _player
        (:Is%20Valid
          (Utilities|IsValid (Variables|/Script/Engine.Character|GetMesh _player)
            (:Is%20Valid
              (return (- (Transformation|GetActorLocation (Variables|Attach|GetAttachPointBGA)) (Transformation|GetWorldLocation (Variables|/Script/Engine.Character|GetMesh (Variables|Default|GetPlayer)))) true))
            (:Is%20Not%20Valid
              _)))
        (:Is%20Not%20Valid
          _)))
    (:Is%20Not%20Valid
      _))
  (return "0, 0, 0" false))
