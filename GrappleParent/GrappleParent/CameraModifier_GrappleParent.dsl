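; CameraModifier_GrappleParent is a camera modifier Blueprint, named as the parent grapple modifier, that can be switched on and off.
; Its Turn On event enables the modifier so it can blend in.
; Its TurnOff event disables it non-immediately, letting it blend out.
; The graph holds only these two toggle events, with no camera adjustment logic of its own.
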(event Custom|Turn%20On
  ;[0.235 1 0.2169] Turn on and off. Enables and disables this modifier, allowing it to blend in and out.
  (CameraModifier|EnableModifier))

(event Custom|TurnOff
  ;[0.235 1 0.2169] Turn on and off. Enables and disables this modifier, allowing it to blend in and out.
  (CameraModifier|DisableModifier false))
