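; Fire-input gameplay ability for the parent grapple weapon, relaying fire button presses and releases to the weapon.
; On activation it caches the avatar's player pawn and its current weapon, cast to B_Weapon_GrappleParent.
; Parallel WaitInputPress and WaitInputRelease tasks call FirePressed and FireReleased on that weapon.
; Release and ability end both remove effects granting the HoldingFireButton tag; this graph never applies one.
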(event Ability|EventActivateAbility
  (bind (_owneractor _avataractor) (Utilities|Struct|BreakGameplayAbilityActorInfo (Ability|GetActorInfo)))
  (|CallParentFunction)
  (bind _asplayer_pawn_athena (Utilities|Casting|CastToPlayerPawn_Athena _avataractor))
  (bind _output_get (Variables|Default|SetPlayer _asplayer_pawn_athena))
  (bind _asb_weapon_grapple_parent (Utilities|Casting|CastToB_Weapon_GrappleParent (Variables|/Script/FortniteGame.FortPawn|GetCurrentWeapon _output_get)))
  (Variables|Default|SetWeapon _asb_weapon_grapple_parent)
  (Utilities|FlowControl|Sequence
    (:then_0
      (Ability|Tasks|WaitInputPress false
        (:OnPress
          (Utilities|FlowControl|Sequence
            (:then_0
              _)
            (:then_1
              _))
          (Class|BWeaponGrappleParent|FirePressed (Variables|Default|GetWeapon)))))
    (:then_1
      (Ability|Tasks|WaitInputRelease false
        (:OnRelease
          (Utilities|FlowControl|Sequence
            (:then_0
              _)
            (:then_1
              (Ability|RemoveGameplayEffectFromOwnerWithGrantedTags "(GameplayTags=((TagName=\"Granted.Athena.GrappleParent.HoldingFireButton\")))"))
            (:then_2
              _))
          (Class|BWeaponGrappleParent|FireReleased (Variables|Default|GetWeapon)))))))

(event Ability|EventFailedToActivatePassiveAbility
  (params (ActorInfo :type Gameplay%20Ability%20Actor%20Info%20Structure))
  (|CallParentFunction))

(event Ability|EventOnEndAbility
  (params (bWasCancelled :type Boolean))
  (|CallParentFunction bWasCancelled)
  (Ability|RemoveGameplayEffectFromOwnerWithGrantedTags "(GameplayTags=((TagName=\"Granted.Athena.GrappleParent.HoldingFireButton\")))"))
