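; Gameplay ability for the grapple weapon's swinging movement mode that toggles a grapple camera modifier on non-authority instances.
; On activation it takes the modifier class from the player's current grapple weapon and finds that modifier on the camera manager, adding one if missing.
; It caches the modifier on itself and the weapon, then turns it on.
; When the ability ends, it turns the cached modifier off.
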(event Ability|EventActivateAbility
  (bind (_owneractor _avataractor) (Utilities|Struct|BreakGameplayAbilityActorInfo (Ability|GetActorInfo)))
  (Ability|HasAuthority
    (:False
      (bind _asplayer_pawn_athena (Utilities|Casting|CastToPlayerPawn_Athena _avataractor))
      (bind _asb_weapon_grapple_parent (Utilities|Casting|CastToB_Weapon_GrappleParent (Variables|/Script/FortniteGame.FortPawn|GetCurrentWeapon _asplayer_pawn_athena)))
      (Variables|Default|SetWeapon _asb_weapon_grapple_parent)
      (Utilities|IsValid (Fortnite|GetFortPlayerControllerfromActor _asplayer_pawn_athena)
        (:Is%20Valid
          (Utilities|IsValid (Variables|/Script/Engine.PlayerController|GetPlayerCameraManager (Fortnite|GetFortPlayerControllerfromActor _asplayer_pawn_athena))
            (:Is%20Valid
              (bind _output_get (Variables|Default|SetCameraManager (Variables|/Script/Engine.PlayerController|GetPlayerCameraManager (Fortnite|GetFortPlayerControllerfromActor _asplayer_pawn_athena))))
              (bind _returnvalue (Game|Player|FindCameraModifierbyClass _output_get (Variables|/GrappleParent/GrappleParent/B_Weapon_GrappleParent.B_Weapon_GrappleParent_C|GetCameraModifierClass (Variables|Default|GetWeapon))))
              (bind _ascamera_modifier_grapple_parent (Utilities|Casting|CastToCameraModifier_GrappleParent _returnvalue)
                (:then
                  (bind _output_get_2 (Variables|Default|SetMod _ascamera_modifier_grapple_parent))
                  (bind _output_get_3 (Variables|/GrappleParent/GrappleParent/B_Weapon_GrappleParent.B_Weapon_GrappleParent_C|SetCameraModifier _output_get_2 (Variables|Default|GetWeapon)))
                  (Class|CameraModifierGrappleParent|TurnOn _output_get_3))
                (:CastFailed
                  (bind _returnvalue_2 (Game|Player|AddNewCameraModifier (Variables|Default|GetCameraManager) (Variables|/GrappleParent/GrappleParent/B_Weapon_GrappleParent.B_Weapon_GrappleParent_C|GetCameraModifierClass (Variables|Default|GetWeapon))))
                  (bind _ascamera_modifier_grapple_parent_2 (Utilities|Casting|CastToCameraModifier_GrappleParent _returnvalue_2))
                  (bind _output_get_4 (Variables|Default|SetMod _ascamera_modifier_grapple_parent_2))
                  (bind _output_get_5 (Variables|/GrappleParent/GrappleParent/B_Weapon_GrappleParent.B_Weapon_GrappleParent_C|SetCameraModifier _output_get_4 (Variables|Default|GetWeapon)))
                  (Class|CameraModifierGrappleParent|TurnOn _output_get_5))))))))))

(event Ability|EventFailedToActivatePassiveAbility
  (params (ActorInfo :type Gameplay%20Ability%20Actor%20Info%20Structure))
  (|CallParentFunction))

(event Ability|EventOnEndAbility
  (params (bWasCancelled :type Boolean))
  (Ability|HasAuthority
    (:False
      (bind _mod (Variables|Default|GetMod)
        (:then
          (Class|CameraModifierGrappleParent|TurnOff _mod))
        (:else)))))
