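; Blueprint grapple swinging component that relays component events to its owning grapple weapon.
; Each event casts the component's owner to B_Weapon_GrappleParent.
; Exiting the ground tells the weapon to stop skidding.
; Reticle updates forward hit, use-blocked, aim-at-target and target position values to the weapon's native-driven reticle update.
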(event EventOnExitGround
  (bind _asb_weapon_grapple_parent (Utilities|Casting|CastToB_Weapon_GrappleParent (Components|GetOwner)))
  (Class|BWeaponGrappleParent|StopSkidding _asb_weapon_grapple_parent))

(event EventUpdateReticle
  (params (bHitSomething :type Boolean) (bUseBlocked :type Boolean) (bIsWeaponAimingAtTarget :type Boolean) (TargetPos :type Vector%20%28by%20ref%29 :default "0, 0, 0"))
  (bind _asb_weapon_grapple_parent (Utilities|Casting|CastToB_Weapon_GrappleParent (Components|GetOwner)))
  (Class|BWeaponGrappleParent|UpdateReticlefromNative _asb_weapon_grapple_parent bHitSomething bUseBlocked bIsWeaponAimingAtTarget TargetPos))
