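; Grapple projectile actor that carries an attached pickup back to its instigator.
; Init attaches the pickup and, on the remote side, calls StartMoving to record the initial direction to the instigator.
; Tick steers projectile velocity toward the instigator on authority while moving.
; Close range or a dot-product overshoot check fires OnDonePullingPickup, places the actor ahead of the instigator, stops it and sets a short lifespan.
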(event Custom|Init
  (params (AttachedPickup :type Actor%20Object%20Reference))
  (Transformation|AttachActorToActor AttachedPickup self "None" "SnapToTarget" "SnapToTarget" "SnapToTarget" true)
  (Variables|Default|SetAttachedPickup AttachedPickup)
  (Networking|SwitchHasAuthority
    (:Remote
      (CallFunction|StartMoving))))

(event Custom|StartMoving
  (Variables|Default|SetMove true)
  (Variables|Default|SetOriginalVel (- (Transformation|GetActorLocation (Variables|Actor|GetInstigator)) (Transformation|GetActorLocation))))

(event EventTick
  (params (DeltaSeconds :type Float%20%28single-precision%29))
  (if (not (Variables|Default|GetDone))
    (Variables|Default|SetVecToPlayer (- (Transformation|GetActorLocation (Variables|Actor|GetInstigator)) (Transformation|GetActorLocation)))
    (Networking|SwitchHasAuthority
      (:Authority
        (if (Variables|Default|GetMove)
          (Projectile|SetVelocity (Variables|Movement|GetProjectileMovementComponent) (* (Math|Vector|Normalize (Variables|Default|GetVecToPlayer) 0.001) 4e+03))))
      (:Remote
        _))
    (if (or (< (Math|Vector|VectorLengthSquared (Variables|Default|GetVecToPlayer)) 5e+02) (< (Math|Vector|DotProduct (Variables|Default|GetVecToPlayer) (Variables|Default|GetOriginalVel)) 0.0))
      (Default|CallOnDonePullingPickup)
      (Variables|Default|SetDone true)
      (Transformation|SetActorLocation (+ (Transformation|GetActorLocation (Variables|Actor|GetInstigator)) (* (Transformation|GetActorForwardVector (Variables|Actor|GetInstigator)) 1e+02)))
      (Projectile|SetVelocity (Variables|Movement|GetProjectileMovementComponent) (Math|Vector|MakeVector 0.0))
      (Actor|SetLifeSpan 3.0))))

(fn ConstructionScript)
